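# Physics asset for the cape on the female medium OutlineLove Blur player cosmetic body.
# Center, left and right cape chains each link seven capsule bodies to a small master body, with capsules growing toward the hem and side capsules offset laterally.
# Swing limits widen down each chain while twist stays tight.
# Unconstrained capsules on pelvis, spine, clavicles, feet and skirt bones supply body shapes; arm, thigh and calf bodies have no shapes.
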PhysicsAsset /BRCosmetics/Characters/Player/Female/Medium/Bodies/F_MED_OutlineLove_Blur/Meshes/F_MED_OutlineLove_Blur_Cape_physics
bodies {
 calf_l []
 calf_r []
 clavicle_l [{shapeType=Capsule, length=6.5623, radius=7.5194, center=<3.9566 -1.0862 -2.1127>, rotation=<rot -77.5566 177.3915 7.7314>}]
 clavicle_r [{shapeType=Capsule, length=6.5623, radius=7.5194, center=<-3.9566 1.0862 2.1127>, rotation=<rot -77.5566 177.3915 -172.2686>}]
 dyn_cape_01 [{shapeType=Capsule, length=11.7556, radius=5.9773, shapeName=dyn_cape_01_capsule, center=<7.15 0 8>, rotation=<rot 90 0 0>}]
 dyn_cape_02 [{shapeType=Capsule, length=10, radius=10, shapeName=dyn_cape_01_capsule, center=<7.15 0 5>, rotation=<rot 90 0 0>}]
 dyn_cape_03 [{shapeType=Capsule, length=10, radius=12, shapeName=dyn_cape_01_capsule, center=<7.15 0 5>, rotation=<rot 90 0 0>}]
 dyn_cape_04 [{shapeType=Capsule, length=10, radius=12, shapeName=dyn_cape_01_capsule, center=<7.15 0 5>, rotation=<rot 90 0 0>}]
 dyn_cape_05 [{shapeType=Capsule, length=10, radius=12, shapeName=dyn_cape_01_capsule, center=<7.15 0 5>, rotation=<rot 90 0 0>}]
 dyn_cape_06 [{shapeType=Capsule, length=10, radius=12, shapeName=dyn_cape_01_capsule, center=<7.15 0 5>, rotation=<rot 90 0 0>}]
 dyn_cape_07 [{shapeType=Capsule, length=10, radius=14, shapeName=dyn_cape_01_capsule, center=<5.8445 0 5>, rotation=<rot 90 0 0>}]
 dyn_cape_MSTR [{shapeType=Capsule, length=0.505, radius=0.505, shapeName=dyn_cape_r_MSTR_capsule, rotation=<rot 0 -90 0>}]
 dyn_cape_l_01 [{shapeType=Capsule, length=11.7556, radius=5.9773, shapeName=dyn_cape_01_capsule, center=<7.15 0 8>, rotation=<rot 90 0 0>}]
 dyn_cape_l_02 [{shapeType=Capsule, length=10, radius=10, shapeName=dyn_cape_01_capsule, center=<7.15 4 5>, rotation=<rot 90 0 0>}]
 dyn_cape_l_03 [{shapeType=Capsule, length=10, radius=12, shapeName=dyn_cape_01_capsule, center=<7.15 7 5>, rotation=<rot 90 0 0>}]
 dyn_cape_l_04 [{shapeType=Capsule, length=10, radius=12, shapeName=dyn_cape_01_capsule, center=<7.15 7 5>, rotation=<rot 90 0 0>}]
 dyn_cape_l_05 [{shapeType=Capsule, length=10, radius=12, shapeName=dyn_cape_01_capsule, center=<7.15 7 5>, rotation=<rot 90 0 0>}]
 dyn_cape_l_06 [{shapeType=Capsule, length=10, radius=12, shapeName=dyn_cape_01_capsule, center=<7.15 7 5>, rotation=<rot 90 0 0>}]
 dyn_cape_l_07 [{shapeType=Capsule, length=10, radius=14, shapeName=dyn_cape_01_capsule, center=<5.8445 10 5>, rotation=<rot 90 0 0>}]
 dyn_cape_l_MSTR [{shapeType=Capsule, length=0.505, radius=0.505, shapeName=dyn_cape_r_MSTR_capsule, rotation=<rot 0 -90 0>}]
 dyn_cape_r_01 [{shapeType=Capsule, length=11.7556, radius=5.9773, shapeName=dyn_cape_01_capsule, center=<7.15 0 8>, rotation=<rot 90 0 0>}]
 dyn_cape_r_02 [{shapeType=Capsule, length=10, radius=10, shapeName=dyn_cape_01_capsule, center=<7.15 -4 5>, rotation=<rot 90 0 0>}]
 dyn_cape_r_03 [{shapeType=Capsule, length=10, radius=12, shapeName=dyn_cape_01_capsule, center=<7.15 -7 5>, rotation=<rot 90 0 0>}]
 dyn_cape_r_04 [{shapeType=Capsule, length=10, radius=12, shapeName=dyn_cape_01_capsule, center=<7.15 -7 5>, rotation=<rot 90 0 0>}]
 dyn_cape_r_05 [{shapeType=Capsule, length=10, radius=12, shapeName=dyn_cape_01_capsule, center=<7.15 -7 5>, rotation=<rot 90 0 0>}]
 dyn_cape_r_06 [{shapeType=Capsule, length=10, radius=12, shapeName=dyn_cape_01_capsule, center=<7.15 -7 5>, rotation=<rot 90 0 0>}]
 dyn_cape_r_07 [{shapeType=Capsule, length=10, radius=14, shapeName=dyn_cape_01_capsule, center=<5.8445 -10 5>, rotation=<rot 90 0 0>}]
 dyn_cape_r_MSTR [{shapeType=Capsule, length=0.505, radius=0.505, shapeName=dyn_cape_r_MSTR_capsule, rotation=<rot 0 -90 0>}]
 dyn_skirt_l_3_02 [{shapeType=Capsule, length=18.1158, radius=7.3127, shapeName=dyn_skirt_r_3_02_capsule, center=<-1.5907 0.9745 0>, rotation=<rot 90 0 0>}]
 dyn_skirt_l_3_04 [{shapeType=Capsule, length=17.2572, radius=7.6976, shapeName=dyn_skirt_l_3_04_capsule, center=<-5.136 1.907 0>, rotation=<rot 90 0 0>}]
 dyn_skirt_l_3_05 [{shapeType=Capsule, length=12.8237, radius=7.6976, shapeName=dyn_skirt_l_3_05_capsule, center=<-1.5907 2.3992 0>, rotation=<rot 90 0 0>}]
 dyn_skirt_l_4_02 [{shapeType=Capsule, length=12.8237, radius=7.6976, shapeName=dyn_skirt_l_4_02_capsule, center=<8.8025 4.9467 0>, rotation=<rot 90 0 0>}]
 dyn_skirt_l_4_04 [{shapeType=Capsule, length=17.6535, radius=8.5572, shapeName=dyn_skirt_l_4_04_capsule, center=<5.2572 1.9685 0>, rotation=<rot 90 0 0>}]
 dyn_skirt_r_3_02 [{shapeType=Capsule, length=18.1158, radius=7.3127, shapeName=dyn_skirt_r_3_02_capsule, center=<-1.5907 0.9745 0>, rotation=<rot 90 0 0>}]
 dyn_skirt_r_3_04 [{shapeType=Capsule, length=17.2572, radius=7.6976, shapeName=dyn_skirt_r_3_04_capsule, center=<-5.136 1.603 0>, rotation=<rot 90 0 0>}]
 dyn_skirt_r_3_05 [{shapeType=Capsule, length=15.6618, radius=7.6976, shapeName=dyn_skirt_r_3_05_capsule, center=<-1.5907 1.7737 0>, rotation=<rot 90 0 0>}]
 dyn_skirt_r_4_02 [{shapeType=Capsule, length=12.8237, radius=7.6976, shapeName=dyn_skirt_r_4_02_capsule, center=<-1.5907 4.5727 0>, rotation=<rot 90 0 0>}]
 dyn_skirt_r_4_04 [{shapeType=Capsule, length=17.4635, radius=8.5572, shapeName=dyn_skirt_r_4_04_capsule, center=<-0.8789 2.3188 0>, rotation=<rot 90 0 0>}]
 foot_l [{shapeType=Capsule, length=11.0765, radius=8.0344, center=<-4.1811 6.0027 -0.338>, rotation=<rot 0.0388 1.4701 -89.3982>}]
 foot_r [{shapeType=Capsule, length=11.0765, radius=8.0344, center=<4.1811 -6.0027 0.338>, rotation=<rot 0.0388 1.4699 90.6018>}]
 lowerarm_l []
 lowerarm_r []
 pelvis [{shapeType=Capsule, length=9.6042, radius=11.1441, center=<-2.4471 -2.5289 -0.1981>}]
 spine_02 [
  {shapeType=Capsule, length=3.5118, radius=8.7417, center=<-2.6885 -0.5337 0>}
  {shapeType=Capsule, length=19.1875, radius=10.8125, center=<-1.8317 1.9362 0>, rotation=<rot 90 0 0>}
 ]
 spine_03 [
  {shapeType=Capsule, length=6.5613, radius=9.1269, center=<0.0574 0.4416 0>}
  {shapeType=Capsule, length=21.097, radius=8.903, center=<-1.2163 0.549 0>, rotation=<rot 90 180 -175.0732>}
 ]
 spine_05 [{shapeType=Capsule, length=12.7001, radius=9.335, center=<0.6969 1.91 0>}]
 thigh_l []
 thigh_r []
 upperarm_l []
 upperarm_r []
}
constraints [
 {bone1Name=dyn_cape_01, bone2Name=dyn_cape_MSTR, swing1LimitDegrees=10, swing2LimitDegrees=5, twistLimitDegrees=5}
 {bone1Name=dyn_cape_02, bone2Name=dyn_cape_01, swing1LimitDegrees=20, swing2LimitDegrees=10, twistLimitDegrees=5}
 {bone1Name=dyn_cape_03, bone2Name=dyn_cape_02, swing1LimitDegrees=45, swing2LimitDegrees=20, twistLimitDegrees=5}
 {bone1Name=dyn_cape_04, bone2Name=dyn_cape_03, swing1LimitDegrees=45, swing2LimitDegrees=45, twistLimitDegrees=5}
 {bone1Name=dyn_cape_05, bone2Name=dyn_cape_04, swing1LimitDegrees=45, swing2LimitDegrees=45, twistLimitDegrees=5}
 {bone1Name=dyn_cape_06, bone2Name=dyn_cape_05, swing1LimitDegrees=45, swing2LimitDegrees=45, twistLimitDegrees=5}
 {bone1Name=dyn_cape_07, bone2Name=dyn_cape_06, swing1LimitDegrees=45, swing2LimitDegrees=45, twistLimitDegrees=5}
 {bone1Name=dyn_cape_l_01, bone2Name=dyn_cape_l_MSTR, swing1LimitDegrees=10, swing2LimitDegrees=5, twistLimitDegrees=5}
 {bone1Name=dyn_cape_l_02, bone2Name=dyn_cape_l_01, swing1LimitDegrees=20, swing2LimitDegrees=10, twistLimitDegrees=5}
 {bone1Name=dyn_cape_l_03, bone2Name=dyn_cape_l_02, swing1LimitDegrees=45, swing2LimitDegrees=20, twistLimitDegrees=5}
 {bone1Name=dyn_cape_l_04, bone2Name=dyn_cape_l_03, swing1LimitDegrees=45, swing2LimitDegrees=45, twistLimitDegrees=5}
 {bone1Name=dyn_cape_l_05, bone2Name=dyn_cape_l_04, swing1LimitDegrees=45, swing2LimitDegrees=45, twistLimitDegrees=5}
 {bone1Name=dyn_cape_l_06, bone2Name=dyn_cape_l_05, swing1LimitDegrees=45, swing2LimitDegrees=45, twistLimitDegrees=5}
 {bone1Name=dyn_cape_l_07, bone2Name=dyn_cape_l_06, swing1LimitDegrees=45, swing2LimitDegrees=45, twistLimitDegrees=5}
 {bone1Name=dyn_cape_r_01, bone2Name=dyn_cape_r_MSTR, swing1LimitDegrees=10, swing2LimitDegrees=5, twistLimitDegrees=5}
 {bone1Name=dyn_cape_r_02, bone2Name=dyn_cape_r_01, swing1LimitDegrees=20, swing2LimitDegrees=10, twistLimitDegrees=5}
 {bone1Name=dyn_cape_r_03, bone2Name=dyn_cape_r_02, swing1LimitDegrees=45, swing2LimitDegrees=20, twistLimitDegrees=5}
 {bone1Name=dyn_cape_r_04, bone2Name=dyn_cape_r_03, swing1LimitDegrees=45, swing2LimitDegrees=45, twistLimitDegrees=5}
 {bone1Name=dyn_cape_r_05, bone2Name=dyn_cape_r_04, swing1LimitDegrees=45, swing2LimitDegrees=45, twistLimitDegrees=5}
 {bone1Name=dyn_cape_r_06, bone2Name=dyn_cape_r_05, swing1LimitDegrees=45, swing2LimitDegrees=45, twistLimitDegrees=5}
 {bone1Name=dyn_cape_r_07, bone2Name=dyn_cape_r_06, swing1LimitDegrees=45, swing2LimitDegrees=45, twistLimitDegrees=5}
]
