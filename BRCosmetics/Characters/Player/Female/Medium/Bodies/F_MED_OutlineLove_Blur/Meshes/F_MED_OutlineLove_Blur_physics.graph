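# Physics asset for the F_MED_OutlineLove_Blur female player body cosmetic, set up for dynamic skirt motion.
# Four skirt bone chains of capsules, two per side, are locked to root at the top, with swing limits widening down each chain and tight twist limits.
# Thigh, calf and foot capsules on free constraints give the leg shapes, and an oversized root capsule below the origin acts as a floor.
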PhysicsAsset /BRCosmetics/Characters/Player/Female/Medium/Bodies/F_MED_OutlineLove_Blur/Meshes/F_MED_OutlineLove_Blur_physics
bodies {
 calf_l [{
   shapeType=Capsule
   length=34.0949
   radius=9.7655
   shapeName=calf_r_capsule
   center=<-20.7112 -2.826 -0.5289>
   rotation=<rot 88.245 179.9996 179.9998>
  }]
 calf_r [{
   shapeType=Capsule
   length=34.5949
   radius=9.2655
   shapeName=calf_r_capsule
   center=<20.7112 2.826 -0.5289>
   rotation=<rot 88.245 179.9996 179.9998>
  }]
 dyn_skirt_l_3_01 [{shapeType=Capsule, length=11.0348, radius=7.3357, shapeName=dyn_skirt_l_3_01_capsule, center=<5.4098 -5.5624 0>}]
 dyn_skirt_l_3_02 [{shapeType=Capsule, length=22.1867, radius=10.8603, shapeName=dyn_skirt_l_3_02_capsule, center=<8.4887 -12 -1.3>, rotation=<rot 0 0 -6.3>}]
 dyn_skirt_l_3_03 [{
   shapeType=Capsule
   length=22.1867
   radius=10.7227
   shapeName=dyn_skirt_l_3_03_capsule
   center=<8.8495 -10.6297 -1.176>
   rotation=<rot 0 0 -6.3>
  }]
 dyn_skirt_l_3_04 [{
   shapeType=Capsule
   length=22.1867
   radius=10.7002
   shapeName=dyn_skirt_l_3_04_capsule
   center=<9.8046 -11.6097 -1.2796>
   rotation=<rot 0 0 -6.3>
  }]
 dyn_skirt_l_3_05 [{
   shapeType=Capsule
   length=21.6867
   radius=10.8817
   shapeName=dyn_skirt_l_3_05_capsule
   center=<7.7552 -10.3097 0.1204>
   rotation=<rot 0 0 -6.3>
  }]
 dyn_skirt_l_4_01 [{shapeType=Capsule, length=11.0348, radius=7.3357, shapeName=dyn_skirt_l_4_01_capsule, center=<5.4098 -5.9281 0>}]
 dyn_skirt_l_4_02 [{shapeType=Capsule, length=19.7439, radius=10.3928, shapeName=dyn_skirt_l_4_02_capsule, center=<7.4512 -10.64 0>}]
 dyn_skirt_l_4_03 [{shapeType=Capsule, length=19.7439, radius=10.5397, shapeName=dyn_skirt_l_4_03_capsule, center=<9.1806 -9.2256 0>}]
 dyn_skirt_l_4_04 [{shapeType=Capsule, length=18.7439, radius=11.9418, shapeName=dyn_skirt_l_4_04_capsule, center=<9.5127 -9.5731 0>}]
 dyn_skirt_r_3_01 [{shapeType=Capsule, length=11.0348, radius=7.3357, shapeName=dyn_skirt_r_3_01_capsule, center=<5.4098 -5.5624 0>}]
 dyn_skirt_r_3_02 [{shapeType=Capsule, length=22.1867, radius=10.8603, shapeName=dyn_skirt_l_3_02_capsule, center=<8.4887 -12 -1.3>, rotation=<rot 0 0 -6.3>}]
 dyn_skirt_r_3_03 [{
   shapeType=Capsule
   length=22.1867
   radius=10.7227
   shapeName=dyn_skirt_l_3_03_capsule
   center=<8.8495 -10.6297 -1.176>
   rotation=<rot 0 0 -6.3>
  }]
 dyn_skirt_r_3_04 [{
   shapeType=Capsule
   length=22.1867
   radius=10.7002
   shapeName=dyn_skirt_l_3_04_capsule
   center=<9.8046 -11.6097 -1.2796>
   rotation=<rot 0 0 -6.3>
  }]
 dyn_skirt_r_3_05 [{
   shapeType=Capsule
   length=21.6867
   radius=10.8817
   shapeName=dyn_skirt_l_3_05_capsule
   center=<7.7552 -10.3097 0.1204>
   rotation=<rot 0 0 -6.3>
  }]
 dyn_skirt_r_4_01 [{shapeType=Capsule, length=11.0348, radius=7.3357, shapeName=dyn_skirt_r_4_01_capsule, center=<5.4098 -5.9281 0>}]
 dyn_skirt_r_4_02 [{shapeType=Capsule, length=19.7439, radius=10.3928, shapeName=dyn_skirt_l_4_02_capsule, center=<7.4512 -10.64 0>}]
 dyn_skirt_r_4_03 [{shapeType=Capsule, length=19.7439, radius=10.5397, shapeName=dyn_skirt_l_4_03_capsule, center=<9.1806 -9.2256 0>}]
 dyn_skirt_r_4_04 [{shapeType=Capsule, length=18.7439, radius=11.9418, shapeName=dyn_skirt_l_4_04_capsule, center=<9.5127 -9.5731 0>}]
 foot_l [{
   shapeType=Capsule
   length=6.5773
   radius=11.1065
   shapeName=foot_r_capsule
   center=<-2.2481 4.5297 -0.2815>
   rotation=<rot 0 6.9743 -89.7415>
  }]
 foot_r [{shapeType=Capsule, length=6.5773, radius=11.1065, shapeName=foot_r_capsule, center=<2.2481 -4.5297 0.2815>, rotation=<rot 0 6.9743 -89.7415>}]
 root [{shapeType=Capsule, length=0.5056, radius=337.6725, shapeName=floor_joint_capsule, center=<2.6954 -3.0549 -337.3088>, rotation=<rot 0 90 -90>}]
 thigh_l [{
   shapeType=Capsule
   length=29.859
   radius=8.229
   shapeName=thigh_r_capsule
   center=<-21.8419 -0.8269 0.6041>
   rotation=<rot 87.6898 179.7339 179.8332>
  }]
 thigh_r [{
   shapeType=Capsule
   length=29.859
   radius=8.229
   shapeName=thigh_r_capsule
   center=<21.8419 0.8269 -0.6041>
   rotation=<rot 87.6898 179.7339 179.8332>
  }]
}
constraints [
 {
  bone1Name=dyn_skirt_l_3_01
  bone2Name=root
  swing1LimitDegrees=2
  swing1Motion=Locked
  swing2LimitDegrees=1
  swing2Motion=Locked
  twistLimitDegrees=1
  twistMotion=Locked
 }
 {bone1Name=dyn_skirt_l_3_02, bone2Name=dyn_skirt_l_3_01, swing1LimitDegrees=10, swing2LimitDegrees=10, twistLimitDegrees=1}
 {bone1Name=dyn_skirt_l_3_03, bone2Name=dyn_skirt_l_3_02, swing1LimitDegrees=15, swing2LimitDegrees=15, twistLimitDegrees=1}
 {bone1Name=dyn_skirt_l_3_04, bone2Name=dyn_skirt_l_3_03, swing1LimitDegrees=17, swing2LimitDegrees=17, twistLimitDegrees=1}
 {bone1Name=dyn_skirt_l_3_05, bone2Name=dyn_skirt_l_3_04, swing1LimitDegrees=20, swing2LimitDegrees=20, twistLimitDegrees=1}
 {
  bone1Name=dyn_skirt_l_4_01
  bone2Name=root
  swing1LimitDegrees=2
  swing1Motion=Locked
  swing2LimitDegrees=1
  swing2Motion=Locked
  twistLimitDegrees=1
  twistMotion=Locked
 }
 {bone1Name=dyn_skirt_l_4_02, bone2Name=dyn_skirt_l_4_01, swing1LimitDegrees=10, swing2LimitDegrees=10, twistLimitDegrees=1}
 {bone1Name=dyn_skirt_l_4_03, bone2Name=dyn_skirt_l_4_02, swing1LimitDegrees=15, swing2LimitDegrees=15, twistLimitDegrees=1}
 {bone1Name=dyn_skirt_l_4_04, bone2Name=dyn_skirt_l_4_03, swing1LimitDegrees=17, swing2LimitDegrees=17, twistLimitDegrees=1}
 {
  bone1Name=dyn_skirt_r_4_01
  bone2Name=root
  swing1LimitDegrees=2
  swing1Motion=Locked
  swing2LimitDegrees=1
  swing2Motion=Locked
  twistLimitDegrees=1
  twistMotion=Locked
 }
 {bone1Name=dyn_skirt_r_4_02, bone2Name=dyn_skirt_r_4_01, swing1LimitDegrees=10, swing2LimitDegrees=10, twistLimitDegrees=1}
 {bone1Name=dyn_skirt_r_4_03, bone2Name=dyn_skirt_r_4_02, swing1LimitDegrees=15, swing2LimitDegrees=15, twistLimitDegrees=1}
 {bone1Name=dyn_skirt_r_4_04, bone2Name=dyn_skirt_r_4_03, swing1LimitDegrees=17, swing2LimitDegrees=17, twistLimitDegrees=1}
 {
  bone1Name=thigh_l
  bone2Name=root
  swing1LimitDegrees=45
  swing1Motion=Free
  swing2LimitDegrees=45
  swing2Motion=Free
  twistLimitDegrees=45
  twistMotion=Free
 }
 {
  bone1Name=calf_l
  bone2Name=thigh_l
  swing1LimitDegrees=45
  swing1Motion=Free
  swing2LimitDegrees=45
  swing2Motion=Free
  twistLimitDegrees=45
  twistMotion=Free
 }
 {
  bone1Name=foot_l
  bone2Name=calf_l
  swing1LimitDegrees=45
  swing1Motion=Free
  swing2LimitDegrees=45
  swing2Motion=Free
  twistLimitDegrees=45
  twistMotion=Free
 }
 {
  bone1Name=thigh_r
  bone2Name=root
  swing1LimitDegrees=45
  swing1Motion=Free
  swing2LimitDegrees=45
  swing2Motion=Free
  twistLimitDegrees=45
  twistMotion=Free
 }
 {
  bone1Name=calf_r
  bone2Name=thigh_r
  swing1LimitDegrees=45
  swing1Motion=Free
  swing2LimitDegrees=45
  swing2Motion=Free
  twistLimitDegrees=45
  twistMotion=Free
 }
 {
  bone1Name=foot_r
  bone2Name=calf_r
  swing1LimitDegrees=45
  swing1Motion=Free
  swing2LimitDegrees=45
  swing2Motion=Free
  twistLimitDegrees=45
  twistMotion=Free
 }
 {
  bone1Name=dyn_skirt_r_3_01
  bone2Name=root
  swing1LimitDegrees=2
  swing1Motion=Locked
  swing2LimitDegrees=1
  swing2Motion=Locked
  twistLimitDegrees=1
  twistMotion=Locked
 }
 {bone1Name=dyn_skirt_r_3_02, bone2Name=dyn_skirt_r_3_01, swing1LimitDegrees=10, swing2LimitDegrees=10, twistLimitDegrees=1}
 {bone1Name=dyn_skirt_r_3_03, bone2Name=dyn_skirt_r_3_02, swing1LimitDegrees=15, swing2LimitDegrees=15, twistLimitDegrees=1}
 {bone1Name=dyn_skirt_r_3_04, bone2Name=dyn_skirt_r_3_03, swing1LimitDegrees=17, swing2LimitDegrees=17, twistLimitDegrees=1}
 {bone1Name=dyn_skirt_r_3_05, bone2Name=dyn_skirt_r_3_04, swing1LimitDegrees=20, swing2LimitDegrees=20, twistLimitDegrees=1}
]
