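# Animation Blueprint for the F_MED_OutlineLove_Blur character body, adding secondary motion to skirt, pendant and cape.
# It copies parent and hair-mesh poses, simulates skirt and pendant with constraints, AnimDynamics and RigidBody, then applies a Control Rig.
# Cached cape variants branch from that pose: backpack RigidBody, cloth-collision bone or RigidBody, AnimDynamics, and trail controllers with a DBNO version.
# Nested bool blends select one from pack, physics-feature and DBNO flags.
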AnimBlueprint /BRCosmetics/Characters/Player/Female/Medium/Bodies/F_MED_OutlineLove_Blur/Meshes/F_MED_OutlineLove_Blur_AnimBP
target_skeleton=&/F_MED_OutlineLove_Blur_Skeleton
graphs [
 AnimGraph {nodes [
   n1 = Blend Poses by bool(
    BlendPose_0=Use cached pose 'animDyn Cache'()
    BlendPose_1=Blend Poses by bool(BlendPose_0=Use cached pose 'dbnoCache'(), BlendPose_1=Use cached pose 'trailCache'(), bActiveValue=Get bIsDBNO()) {blendType=Linear}
    bActiveValue=Get bAnimDynamicsEnabled() -> (bAnimDynamicsEnabled:bool)
   ) {blendType=Linear}
   n2 = Use cached pose 'RbanCache'()
   n3 = Get bRigidBodyEnabled()
   Output Pose(Result=Blend Poses by bool(
     BlendPose_0=Use cached pose 'backpackRban'()
     BlendPose_1=Blend Poses by bool(BlendPose_0=@n2, BlendPose_1=@n1, bActiveValue=@n3) {blendType=Linear}
     BlendTime_0=0
     BlendTime_1=0
     bActiveValue=Get isWearingPack() -> (isWearingPack:bool)
    ) {blendType=Linear}) {layerGroup=None}
   n4 = "Constraint - dyn_pendant_01 (PPPP)"!(
    ComponentPose=Local To Component(LocalPose=Use cached pose 'SkirtDynPose'())
    Alpha=1
    ConstraintWeights_0=0.6
    ConstraintWeights_1:float=0.3413
    ConstraintWeights_2:float=0.035
    ConstraintWeights_3:float=0.0237
   ) -> (Pose) {
    boneToModify {boneName=dyn_pendant_01}
    constraintSetup [
     {targetBone {boneName=dyn_pendant_01}}
     {targetBone {boneName=spine_05}}
     {targetBone {boneName=spine_04}}
     {targetBone {boneName=clavicle_l}}
    ]
    constraintWeights [1,1,1,1]
   }
   n5 = Anim Dynamics - Bone: dyn_pendant_01(ComponentPose=@n4, Alpha=1, ExternalForce="0,0,0") {
    angularDampingOverride=1
    bOverrideAngularDamping=true
    bOverrideLinearDamping=true
    bUseGravityOverride=true
    linearDampingOverride=0.9999
    simulationSpace=RootRelative
    boundBone {boneName=dyn_pendant_01}
    componentAppliedLinearAccClamp=<3000 3000 3000>
    componentLinearAccScale=<0.1 0.1 0.1>
    componentLinearVelScale=<0.2 0.2 0.2>
    gravityOverride {z=-980}
    physicsBodyDefinitions [{
      boundBone {boneName=dyn_pendant_01}
      boxExtents=<10 10 10>
      constraintSetup {angularLimitsMax=<0 -5 5>, angularLimitsMin=<0 -20 -5>, angularTarget=<0 0 0>}
      localJointOffset=<1 0 0>
     }]
    relativeSpaceBone {boneName=spine_05}
   }
   n6 = Anim Dynamics - Bone: dyn_pendant_02(ComponentPose=@n5, Alpha=1, ExternalForce="0,0,0") {
    angularDampingOverride=1
    bOverrideAngularDamping=true
    bOverrideLinearDamping=true
    bUseGravityOverride=true
    linearDampingOverride=0.9999
    simulationSpace=RootRelative
    boundBone {boneName=dyn_pendant_02}
    componentAppliedLinearAccClamp=<3000 3000 3000>
    componentLinearAccScale=<0.1 0.1 0.1>
    componentLinearVelScale=<0.2 0.2 0.2>
    gravityOverride {z=-980}
    physicsBodyDefinitions [{
      boundBone {boneName=dyn_pendant_02}
      boxExtents=<10 10 10>
      constraintSetup {angularLimitsMax=<0 0 5>, angularLimitsMin=<0 -25 -5>, angularTarget=<0 0 0>}
      localJointOffset=<1 0 0>
     }]
    relativeSpaceBone {boneName=spine_05}
   }
   n7 = Control Rig!(
    Source=Component To Local(ComponentPose=@n6)
    bAlphaBoolEnabled:bool=True
    isBaseBallSliding:double=0
    isKneeSliding:double=0
    isSkyDiving:double=0
    MoveBack:double=0
    MoveFront:double=0
    isOnGrindRail:double=0
    isMovingForward:double=0
    isStrafeLeft:double=0
    isStrafeRight:double=0
    CrouchBool:bool=Get bIsCrouching()
   ) -> (Pose) {
    alphaInputType=Bool
    controlRigAssetReference {blueprintRigClass=/BRCosmetics/Characters/Player/Female/Medium/Bodies/F_MED_OutlineLove/Meshes/F_MED_OutlineLove_CtrlRig.F_MED_OutlineLove_CtrlRig_C}
   }
   Save cached pose 'Start_Cache'(Pose=@n7) {cacheName=Start_Cache}
   n8 = Get TrailControllerRot() -> (TrailControllerRot:Rotator)
   n9 = "Transform (Modify) Bone - Bone: dyn_cape_l_01"(ComponentPose=Local To Component(LocalPose=Use cached pose 'Start_Cache'()), Alpha=1, Translation="0,0,0", Rotation=@n8, Scale="1,1,1") {rotationMode=BMM_Additive, rotationSpace=BCS_BoneSpace, alphaBoolBlend {blendOption=Linear}, boneToModify {boneName=dyn_cape_l_01}}
   n10 = "Transform (Modify) Bone - Bone: dyn_cape_01"(ComponentPose=@n9, Alpha=1, Translation="0,0,0", Rotation=@n8, Scale="1,1,1") {rotationMode=BMM_Additive, rotationSpace=BCS_BoneSpace, alphaBoolBlend {blendOption=Linear}, boneToModify {boneName=dyn_cape_01}}
   n11 = "Transform (Modify) Bone - Bone: dyn_cape_r_01"(ComponentPose=@n10, Alpha=1, Translation="0,0,0", Rotation=@n8, Scale="1,1,1") {rotationMode=BMM_Additive, rotationSpace=BCS_BoneSpace, alphaBoolBlend {blendOption=Linear}, boneToModify {boneName=dyn_cape_r_01}}
   n12 = Get BwdMovement() -> (BwdMovement:double)
   n13 = Trail controller - Bone: dyn_cape_l_07(ComponentPose=@n11, Alpha=@n12) {
    bLimitRotation=true
    bUsePlanarLimit=true
    chainLength=5
    debugLifeTime=0.1
    maxDeltaTime=0.0333
    alphaScaleBiasClamp {bInterpResult=true}
    baseJoint {boneName=backpack_root}
    planarLimits [{drivingBone {boneName=backpack_root}, planeTransform {location=<0 0 0>, rotation=<rot 0 0 -90>, scale=<1 1 1>}}]
    rotationLimits [{limitMax=<180 180 180>}, {limitMax=<180 180 180>}, {limitMax=<180 180 180>}, {limitMax=<180 180 180>}, {limitMax=<180 180 180>}]
    rotationOffsets [<0 0 0>, <0 0 0>, <0 0 0>, <0 0 0>, <0 0 0>]
    trailBone {boneName=dyn_cape_l_07}
    trailRelaxationSpeed {editorCurveData {keys [<at 0 23 RCIM_Linear>,<at 0.284 19.3799 RCIM_Linear>,<at 0.6554 8.5924 RCIM_Linear>,<at 1 5 RCIM_Linear>]}}
   }
   n14 = Trail controller - Bone: dyn_cape_07(ComponentPose=@n13, Alpha=@n12) {
    bLimitRotation=true
    bUsePlanarLimit=true
    chainLength=5
    debugLifeTime=0.1
    maxDeltaTime=0.0333
    alphaScaleBiasClamp {bInterpResult=true}
    baseJoint {boneName=backpack_root}
    planarLimits [{drivingBone {boneName=backpack_root}, planeTransform {location=<0 0 0>, rotation=<rot 0 0 -90>, scale=<1 1 1>}}]
    rotationLimits [
     {limitMax=<180 180 180>}
     {limitMax=<180 180 180>}
     {limitMax=<180 90 180>, limitMin=<-180 -90 -180>}
     {limitMax=<180 180 180>}
     {limitMax=<180 180 180>}
    ]
    rotationOffsets [<0 0 0>, <0 0 0>, <0 0 0>, <0 0 0>, <0 0 0>]
    trailBone {boneName=dyn_cape_07}
    trailRelaxationSpeed {editorCurveData {keys [<at 0 23 RCIM_Linear>,<at 0.284 19.3799 RCIM_Linear>,<at 0.6554 8.5924 RCIM_Linear>,<at 1 5 RCIM_Linear>]}}
   }
   n15 = Trail controller - Bone: dyn_cape_r_07(ComponentPose=@n14, Alpha=@n12) {
    bLimitRotation=true
    bUsePlanarLimit=true
    chainLength=5
    debugLifeTime=0.1
    maxDeltaTime=0.0333
    alphaScaleBiasClamp {bInterpResult=true}
    baseJoint {boneName=backpack_root}
    planarLimits [{drivingBone {boneName=backpack_root}, planeTransform {location=<0 0 0>, rotation=<rot 0 0 -90>, scale=<1 1 1>}}]
    rotationLimits [
     {limitMax=<180 180 180>}
     {limitMax=<180 180 180>}
     {limitMax=<180 90 180>, limitMin=<-180 -90 -180>}
     {limitMax=<180 180 180>}
     {limitMax=<180 180 180>}
    ]
    rotationOffsets [<0 0 0>, <0 0 0>, <0 0 0>, <0 0 0>, <0 0 0>]
    trailBone {boneName=dyn_cape_r_07}
    trailRelaxationSpeed {editorCurveData {keys [<at 0 23 RCIM_Linear>,<at 0.284 19.3799 RCIM_Linear>,<at 0.6554 8.5924 RCIM_Linear>,<at 1 5 RCIM_Linear>]}}
   }
   Save cached pose 'trailCache'(Pose=Component To Local(ComponentPose=@n15)) {cacheName=trailCache}
   n16 = "Transform (Modify) Bone - Bone: dyn_cape_l_01"(ComponentPose=Local To Component(LocalPose=Use cached pose 'Start_Cache'()), Alpha=1, Translation="0,0,0", Rotation="-10,0,0", Scale="1,1,1") {
    rotationMode=BMM_Additive
    rotationSpace=BCS_BoneSpace
    alphaBoolBlend {blendOption=Linear}
    boneToModify {boneName=dyn_cape_l_01}
    rotation {pitch=-10}
   }
   n17 = "Transform (Modify) Bone - Bone: dyn_cape_01"(ComponentPose=@n16, Alpha=1, Translation="0,0,0", Rotation="-7,0,0", Scale="1,1,1") {
    rotationMode=BMM_Additive
    rotationSpace=BCS_BoneSpace
    alphaBoolBlend {blendOption=Linear}
    boneToModify {boneName=dyn_cape_01}
    rotation {pitch=-7}
   }
   n18 = "Transform (Modify) Bone - Bone: dyn_cape_r_01"(ComponentPose=@n17, Alpha=1, Translation="0,0,0", Rotation="0,0,0", Scale="1,1,1") {rotationMode=BMM_Additive, rotationSpace=BCS_BoneSpace, alphaBoolBlend {blendOption=Linear}, boneToModify {boneName=dyn_cape_r_01}}
   n19 = "Transform (Modify) Bone - Bone: dyn_cape_l_03"(ComponentPose=@n18, Alpha=1, Translation="0,0,0", Rotation="-10,0,0", Scale="1,1,1") {
    rotationMode=BMM_Additive
    rotationSpace=BCS_BoneSpace
    alphaBoolBlend {blendOption=Linear}
    boneToModify {boneName=dyn_cape_l_03}
    rotation {pitch=-10}
   }
   n20 = "Transform (Modify) Bone - Bone: dyn_cape_03"(ComponentPose=@n19, Alpha=1, Translation="0,0,0", Rotation="-7,0,0", Scale="1,1,1") {
    rotationMode=BMM_Additive
    rotationSpace=BCS_BoneSpace
    alphaBoolBlend {blendOption=Linear}
    boneToModify {boneName=dyn_cape_03}
    rotation {pitch=-7}
   }
   n21 = "Transform (Modify) Bone - Bone: dyn_cape_r_03"(ComponentPose=@n20, Alpha=1, Translation="0,0,0", Rotation="-3,0,0", Scale="1,1,1") {
    rotationMode=BMM_Additive
    rotationSpace=BCS_BoneSpace
    alphaBoolBlend {blendOption=Linear}
    boneToModify {boneName=dyn_cape_r_03}
    rotation {pitch=-3}
   }
   n22 = "Transform (Modify) Bone - Bone: dyn_cape_l_04"(ComponentPose=@n21, Alpha=1, Translation="0,0,0", Rotation="-10,0,0", Scale="1,1,1") {
    rotationMode=BMM_Additive
    rotationSpace=BCS_BoneSpace
    alphaBoolBlend {blendOption=Linear}
    boneToModify {boneName=dyn_cape_l_04}
    rotation {pitch=-10}
   }
   n23 = "Transform (Modify) Bone - Bone: dyn_cape_04"(ComponentPose=@n22, Alpha=1, Translation="0,0,0", Rotation="-10,0,0", Scale="1,1,1") {
    rotationMode=BMM_Additive
    rotationSpace=BCS_BoneSpace
    alphaBoolBlend {blendOption=Linear}
    boneToModify {boneName=dyn_cape_04}
    rotation {pitch=-10}
   }
   n24 = "Transform (Modify) Bone - Bone: dyn_cape_r_04"(ComponentPose=@n23, Alpha=1, Translation="0,0,0", Rotation="-20,0,0", Scale="1,1,1") {
    rotationMode=BMM_Additive
    rotationSpace=BCS_BoneSpace
    alphaBoolBlend {blendOption=Linear}
    boneToModify {boneName=dyn_cape_r_04}
    rotation {pitch=-20}
   }
   n25 = Get BwdMovement()
   n26 = Trail controller - Bone: dyn_cape_l_07(ComponentPose=@n24, Alpha=@n25) {
    bActorSpaceFakeVel=true
    bShowTrailLocation=true
    bUsePlanarLimit=true
    chainLength=7
    debugLifeTime=0.1
    maxDeltaTime=0.0333
    relaxationSpeedScale=2
    alphaScaleBiasClamp {bInterpResult=true, interpSpeedDecreasing=25, interpSpeedIncreasing=25}
    baseJoint {boneName=backpack_root}
    rotationLimits [
     {limitMax=<180 180 180>}
     {limitMax=<180 180 180>}
     {limitMax=<180 180 180>}
     {limitMax=<180 180 180>}
     {limitMax=<180 180 180>}
     {limitMax=<180 180 180>}
     {limitMax=<180 180 180>}
    ]
    rotationOffsets [<0 0 0>, <0 0 0>, <0 0 0>, <0 0 0>, <0 0 0>, <0 0 0>, <0 0 0>]
    trailBone {boneName=dyn_cape_l_07}
    trailRelaxationSpeed {editorCurveData {keys [<at 0 23 RCIM_Linear>,<at 0.284 19.3799 RCIM_Linear>,<at 0.6554 8.5924 RCIM_Linear>,<at 1 5 RCIM_Linear>]}}
   }
   n27 = Trail controller - Bone: dyn_cape_07(ComponentPose=@n26, Alpha=@n25) {
    bActorSpaceFakeVel=true
    bShowTrailLocation=true
    bUsePlanarLimit=true
    chainLength=7
    debugLifeTime=0.1
    maxDeltaTime=0.0333
    relaxationSpeedScale=2
    alphaScaleBiasClamp {bInterpResult=true, interpSpeedDecreasing=25, interpSpeedIncreasing=25}
    baseJoint {boneName=backpack_root}
    rotationLimits [
     {limitMax=<180 180 180>}
     {limitMax=<180 180 180>}
     {limitMax=<180 180 180>}
     {limitMax=<180 180 180>}
     {limitMax=<180 180 180>}
     {limitMax=<180 180 180>}
     {limitMax=<180 180 180>}
    ]
    rotationOffsets [<0 0 0>, <0 0 0>, <0 0 0>, <0 0 0>, <0 0 0>, <0 0 0>, <0 0 0>]
    trailBone {boneName=dyn_cape_07}
    trailRelaxationSpeed {editorCurveData {keys [<at 0 23 RCIM_Linear>,<at 0.284 19.3799 RCIM_Linear>,<at 0.6554 8.5924 RCIM_Linear>,<at 1 5 RCIM_Linear>]}}
   }
   n28 = Trail controller - Bone: dyn_cape_r_07(ComponentPose=@n27, Alpha=@n25) {
    bActorSpaceFakeVel=true
    bShowTrailLocation=true
    bUsePlanarLimit=true
    chainLength=7
    debugLifeTime=0.1
    maxDeltaTime=0.0333
    relaxationSpeedScale=2
    alphaScaleBiasClamp {bInterpResult=true, interpSpeedDecreasing=25, interpSpeedIncreasing=25}
    baseJoint {boneName=backpack_root}
    rotationLimits [
     {limitMax=<180 180 180>}
     {limitMax=<180 180 180>}
     {limitMax=<180 180 180>}
     {limitMax=<180 180 180>}
     {limitMax=<180 180 180>}
     {limitMax=<180 180 180>}
     {limitMax=<180 180 180>}
    ]
    rotationOffsets [<0 0 0>, <0 0 0>, <0 0 0>, <0 0 0>, <0 0 0>, <0 0 0>, <0 0 0>]
    trailBone {boneName=dyn_cape_r_07}
    trailRelaxationSpeed {editorCurveData {keys [<at 0 23 RCIM_Linear>,<at 0.284 19.3799 RCIM_Linear>,<at 0.6554 8.5924 RCIM_Linear>,<at 1 5 RCIM_Linear>]}}
   }
   Save cached pose 'dbnoCache'(Pose=Component To Local(ComponentPose=@n28)) {cacheName=dbnoCache}
   n29 = Anim Dynamics - Bone: dyn_cape_01(ComponentPose=Local To Component(LocalPose=Use cached pose 'Start_Cache'()), Alpha=1, ExternalForce="0,0,0") {
    angularDampingOverride=1
    angularSpringConstant=10000
    bAngularSpring=true
    bOverrideAngularDamping=true
    bUseGravityOverride=true
    gravityScale=0.35
    simulationSpace=BoneRelative
    alphaBoolBlend {blendOption=Linear}
    boundBone {boneName=dyn_cape_01}
    componentAppliedLinearAccClamp=<3000 3000 3000>
    componentLinearAccScale=<2 2 2>
    componentLinearVelScale=<2 2 2>
    gravityOverride {z=-980}
    physicsBodyDefinitions [{
      boundBone {boneName=dyn_cape_01}
      boxExtents=<10 40 40>
      constraintSetup {angularLimitsMax=<0 5 15>, angularLimitsMin=<0 -5 -15>, angularTarget=<1 0 0>}
      localJointOffset=<5 0 0>
     }]
    relativeSpaceBone {boneName=backpack_root}
   }
   n30 = Anim Dynamics - Bone: dyn_cape_03(ComponentPose=@n29, Alpha=1, ExternalForce="0,0,0") {
    angularDampingOverride=1
    angularSpringConstant=10000
    bAngularSpring=true
    bOverrideAngularDamping=true
    bUseGravityOverride=true
    gravityScale=0.35
    simulationSpace=BoneRelative
    alphaBoolBlend {blendOption=Linear}
    boundBone {boneName=dyn_cape_03}
    componentAppliedLinearAccClamp=<3000 3000 3000>
    componentLinearAccScale=<2 2 2>
    componentLinearVelScale=<2 2 2>
    gravityOverride {z=-980}
    physicsBodyDefinitions [{
      boundBone {boneName=dyn_cape_03}
      boxExtents=<10 40 40>
      constraintSetup {angularLimitsMax=<0 5 15>, angularLimitsMin=<0 -40 -15>, angularTarget=<1 0 0>}
      localJointOffset=<5 0 0>
     }]
    relativeSpaceBone {boneName=backpack_root}
   }
   n31 = Anim Dynamics - Bone: dyn_cape_05(ComponentPose=@n30, Alpha=1, ExternalForce="0,0,0") {
    angularDampingOverride=1
    angularSpringConstant=10000
    bAngularSpring=true
    bOverrideAngularDamping=true
    bUseGravityOverride=true
    gravityScale=0.35
    simulationSpace=BoneRelative
    alphaBoolBlend {blendOption=Linear}
    boundBone {boneName=dyn_cape_05}
    componentAppliedLinearAccClamp=<3000 3000 3000>
    componentLinearAccScale=<2 2 2>
    componentLinearVelScale=<2 2 2>
    gravityOverride {z=-980}
    physicsBodyDefinitions [{
      boundBone {boneName=dyn_cape_05}
      boxExtents=<10 40 40>
      constraintSetup {angularLimitsMax=<0 5 15>, angularLimitsMin=<0 -40 -15>, angularTarget=<1 0 0>}
      localJointOffset=<5 0 0>
     }]
    relativeSpaceBone {boneName=backpack_root}
   }
   n32 = Anim Dynamics - Bone: dyn_cape_06(ComponentPose=@n31, Alpha=1, ExternalForce="0,0,0") {
    angularDampingOverride=1
    angularSpringConstant=10000
    bAngularSpring=true
    bOverrideAngularDamping=true
    bUseGravityOverride=true
    gravityScale=0.35
    simulationSpace=BoneRelative
    alphaBoolBlend {blendOption=Linear}
    boundBone {boneName=dyn_cape_06}
    componentAppliedLinearAccClamp=<3000 3000 3000>
    componentLinearAccScale=<2 2 2>
    componentLinearVelScale=<2 2 2>
    gravityOverride {z=-980}
    physicsBodyDefinitions [{
      boundBone {boneName=dyn_cape_06}
      boxExtents=<10 40 40>
      constraintSetup {angularLimitsMax=<0 5 15>, angularLimitsMin=<0 -40 -15>, angularTarget=<1 0 0>}
      localJointOffset=<5 0 0>
     }]
    relativeSpaceBone {boneName=backpack_root}
   }
   n33 = Anim Dynamics - Bone: dyn_cape_r_01(ComponentPose=@n32, Alpha=1, ExternalForce="0,0,0") {
    angularDampingOverride=1
    angularSpringConstant=10000
    bAngularSpring=true
    bOverrideAngularDamping=true
    bUseGravityOverride=true
    gravityScale=0.35
    simulationSpace=BoneRelative
    alphaBoolBlend {blendOption=Linear}
    boundBone {boneName=dyn_cape_r_01}
    componentAppliedLinearAccClamp=<3000 3000 3000>
    componentLinearAccScale=<2 2 2>
    componentLinearVelScale=<2 2 2>
    gravityOverride {z=-980}
    physicsBodyDefinitions [{
      boundBone {boneName=dyn_cape_r_01}
      boxExtents=<10 40 40>
      constraintSetup {angularLimitsMax=<0 5 15>, angularLimitsMin=<0 -5 -15>, angularTarget=<1 0 0>}
      localJointOffset=<5 0 0>
     }]
    relativeSpaceBone {boneName=backpack_root}
   }
   n34 = Anim Dynamics - Bone: dyn_cape_r_03(ComponentPose=@n33, Alpha=1, ExternalForce="0,0,0") {
    angularDampingOverride=1
    angularSpringConstant=10000
    bAngularSpring=true
    bOverrideAngularDamping=true
    bUseGravityOverride=true
    gravityScale=0.35
    simulationSpace=BoneRelative
    alphaBoolBlend {blendOption=Linear}
    boundBone {boneName=dyn_cape_r_03}
    componentAppliedLinearAccClamp=<3000 3000 3000>
    componentLinearAccScale=<2 2 2>
    componentLinearVelScale=<2 2 2>
    gravityOverride {z=-980}
    physicsBodyDefinitions [{
      boundBone {boneName=dyn_cape_r_03}
      boxExtents=<10 40 40>
      constraintSetup {angularLimitsMax=<0 5 15>, angularLimitsMin=<0 -40 -15>, angularTarget=<1 0 0>}
      localJointOffset=<5 0 0>
     }]
    relativeSpaceBone {boneName=backpack_root}
   }
   n35 = Anim Dynamics - Bone: dyn_cape_r_05(ComponentPose=@n34, Alpha=1, ExternalForce="0,0,0") {
    angularDampingOverride=1
    angularSpringConstant=10000
    bAngularSpring=true
    bOverrideAngularDamping=true
    bUseGravityOverride=true
    gravityScale=0.35
    simulationSpace=BoneRelative
    alphaBoolBlend {blendOption=Linear}
    boundBone {boneName=dyn_cape_r_05}
    componentAppliedLinearAccClamp=<3000 3000 3000>
    componentLinearAccScale=<2 2 2>
    componentLinearVelScale=<2 2 2>
    gravityOverride {z=-980}
    physicsBodyDefinitions [{
      boundBone {boneName=dyn_cape_r_05}
      boxExtents=<10 40 40>
      constraintSetup {angularLimitsMax=<0 5 15>, angularLimitsMin=<0 -40 -15>, angularTarget=<1 0 0>}
      localJointOffset=<5 0 0>
     }]
    relativeSpaceBone {boneName=backpack_root}
   }
   n36 = Anim Dynamics - Bone: dyn_cape_r_06(ComponentPose=@n35, Alpha=1, ExternalForce="0,0,0") {
    angularDampingOverride=1
    angularSpringConstant=10000
    bAngularSpring=true
    bOverrideAngularDamping=true
    bUseGravityOverride=true
    gravityScale=0.35
    simulationSpace=BoneRelative
    alphaBoolBlend {blendOption=Linear}
    boundBone {boneName=dyn_cape_r_06}
    componentAppliedLinearAccClamp=<3000 3000 3000>
    componentLinearAccScale=<2 2 2>
    componentLinearVelScale=<2 2 2>
    gravityOverride {z=-980}
    physicsBodyDefinitions [{
      boundBone {boneName=dyn_cape_r_06}
      boxExtents=<10 40 40>
      constraintSetup {angularLimitsMax=<0 5 15>, angularLimitsMin=<0 -40 -15>, angularTarget=<1 0 0>}
      localJointOffset=<5 0 0>
     }]
    relativeSpaceBone {boneName=backpack_root}
   }
   n37 = Anim Dynamics - Bone: dyn_cape_l_01(ComponentPose=@n36, Alpha=1, ExternalForce="0,0,0") {
    angularDampingOverride=1
    angularSpringConstant=10000
    bAngularSpring=true
    bOverrideAngularDamping=true
    bUseGravityOverride=true
    gravityScale=0.35
    simulationSpace=BoneRelative
    alphaBoolBlend {blendOption=Linear}
    boundBone {boneName=dyn_cape_l_01}
    componentAppliedLinearAccClamp=<3000 3000 3000>
    componentLinearAccScale=<2 2 2>
    componentLinearVelScale=<2 2 2>
    gravityOverride {z=-980}
    physicsBodyDefinitions [{
      boundBone {boneName=dyn_cape_l_01}
      boxExtents=<10 40 40>
      constraintSetup {angularLimitsMax=<0 5 15>, angularLimitsMin=<0 -5 -15>, angularTarget=<1 0 0>}
      localJointOffset=<5 0 0>
     }]
    relativeSpaceBone {boneName=backpack_root}
   }
   n38 = Anim Dynamics - Bone: dyn_cape_l_03(ComponentPose=@n37, Alpha=1, ExternalForce="0,0,0") {
    angularDampingOverride=1
    angularSpringConstant=10000
    bAngularSpring=true
    bOverrideAngularDamping=true
    bUseGravityOverride=true
    gravityScale=0.35
    simulationSpace=BoneRelative
    alphaBoolBlend {blendOption=Linear}
    boundBone {boneName=dyn_cape_l_03}
    componentAppliedLinearAccClamp=<3000 3000 3000>
    componentLinearAccScale=<2 2 2>
    componentLinearVelScale=<2 2 2>
    gravityOverride {z=-980}
    physicsBodyDefinitions [{
      boundBone {boneName=dyn_cape_l_03}
      boxExtents=<10 40 40>
      constraintSetup {angularLimitsMax=<0 5 15>, angularLimitsMin=<0 -40 -15>, angularTarget=<1 0 0>}
      localJointOffset=<5 0 0>
     }]
    relativeSpaceBone {boneName=backpack_root}
   }
   n39 = Anim Dynamics - Bone: dyn_cape_l_05(ComponentPose=@n38, Alpha=1, ExternalForce="0,0,0") {
    angularDampingOverride=1
    angularSpringConstant=10000
    bAngularSpring=true
    bOverrideAngularDamping=true
    bUseGravityOverride=true
    gravityScale=0.35
    simulationSpace=BoneRelative
    alphaBoolBlend {blendOption=Linear}
    boundBone {boneName=dyn_cape_l_05}
    componentAppliedLinearAccClamp=<3000 3000 3000>
    componentLinearAccScale=<2 2 2>
    componentLinearVelScale=<2 2 2>
    gravityOverride {z=-980}
    physicsBodyDefinitions [{
      boundBone {boneName=dyn_cape_l_05}
      boxExtents=<10 40 40>
      constraintSetup {angularLimitsMax=<0 5 15>, angularLimitsMin=<0 -40 -15>, angularTarget=<1 0 0>}
      localJointOffset=<5 0 0>
     }]
    relativeSpaceBone {boneName=backpack_root}
   }
   n40 = Anim Dynamics - Bone: dyn_cape_l_06(ComponentPose=@n39, Alpha=1, ExternalForce="0,0,0") {
    angularDampingOverride=1
    angularSpringConstant=10000
    bAngularSpring=true
    bOverrideAngularDamping=true
    bUseGravityOverride=true
    gravityScale=0.35
    simulationSpace=BoneRelative
    alphaBoolBlend {blendOption=Linear}
    boundBone {boneName=dyn_cape_l_06}
    componentAppliedLinearAccClamp=<3000 3000 3000>
    componentLinearAccScale=<2 2 2>
    componentLinearVelScale=<2 2 2>
    gravityOverride {z=-980}
    physicsBodyDefinitions [{
      boundBone {boneName=dyn_cape_l_06}
      boxExtents=<10 40 40>
      constraintSetup {angularLimitsMax=<0 5 15>, angularLimitsMin=<0 -40 -15>, angularTarget=<1 0 0>}
      localJointOffset=<5 0 0>
     }]
    relativeSpaceBone {boneName=backpack_root}
   }
   Save cached pose 'animDyn Cache'(Pose=Component To Local(ComponentPose=@n40)) {cacheName=animDyn Cache}
   n41 = RigidBody!(
    ComponentPose=Local To Component(LocalPose=Use cached pose 'Start_Cache'())
    Alpha
    OverrideWorldGravity:Vector="0,0,0"
    ExternalForce
    SimSpaceSettings:Sim Space Settings Structure=(WorldAlpha=0,VelocityScaleZ=1,DampingAlpha=1,MaxLinearVelocity=10000,MaxAngularVelocity=10000,MaxLinearAcceleration=10000,MaxAngularAcceleration=10000,ExternalLinearDragV=(X=0,Y=0,Z=0),ExternalLinearVelocity=(X=0,Y=0,Z=0),ExternalAngularVelocity=(X=0,Y=0,Z=0))
   ) -> (Pose) {
    bOverrideWorldGravity=true
    overridePhysicsAsset=&/F_MED_OutlineLove_Blur_CapePack_physics
    simulationSpace=BaseBoneSpace
    componentAppliedLinearAccClamp=<3000 3000 3000>
   }
   n42 = "Transform (Modify) Bone - Bone: dyn_cape_l_01"(ComponentPose=@n41, Alpha=1, Translation="0,-2,0", Rotation="0,0,5", Scale="1,1,1") {rotationMode=BMM_Additive, translationMode=BMM_Additive, boneToModify {boneName=dyn_cape_l_01}, rotation {roll=5}, translation {y=-2}}
   Save cached pose 'backpackRban'(Pose=Component To Local(ComponentPose=@n42)) {cacheName=backpackRban}
   n43 = "Constraint - dyn_hair_upper_cape_collision (PT)"!(ComponentPose=Local To Component(LocalPose=Use cached pose 'Start_Cache'()), Alpha=1, ConstraintWeights_0=1, ConstraintWeights_1:float=0) -> (Pose) {
    alphaBoolBlend {blendOption=Linear}
    boneToModify {boneName=dyn_hair_upper_cape_collision}
    constraintSetup [{targetBone {boneName=dyn_hair_bk_04}}, {transformType=Translation, targetBone {boneName=spine_05}}]
    constraintWeights [1,1]
   }
   n44 = "Transform (Modify) Bone - Bone: dyn_hair_upper_cape_collision"(
    ComponentPose="Transform (Modify) Bone - Bone: dyn_hair_upper_cape_collision"(
     ComponentPose="Transform (Modify) Bone - Bone: dyn_hair_upper_cape_collision"(ComponentPose=@n43, Alpha=1, Translation="0,4,0", Rotation="0,0,0", Scale="1,1,1") {translationMode=BMM_Additive, boneToModify {boneName=dyn_hair_upper_cape_collision}, translation {y=4}}
     Alpha=Get isSkyDivingF() -> (isSkyDivingF:double)
     Translation="0,-4,-5"
     Rotation="0,0,0"
     Scale="1,1,1"
    ) {translationMode=BMM_Additive, boneToModify {boneName=dyn_hair_upper_cape_collision}, translation {y=-4, z=-5}}
    Alpha=Get isIdleInFE() -> (isIdleInFE:double)
    Translation="0,-5.5,-10"
    Rotation="90,10,0"
    Scale="1,1,1"
   ) {
    rotationMode=BMM_Additive
    translationMode=BMM_Additive
    boneToModify {boneName=dyn_hair_upper_cape_collision}
    rotation {pitch=90, yaw=10}
    translation {y=-5.5, z=-10}
   }
   n45 = RigidBody(
    ComponentPose=Local To Component(LocalPose=Use cached pose 'Start_Cache'())
    OverrideWorldGravity="0,0,0"
    SimSpaceSettings=(WorldAlpha=0,VelocityScaleZ=1,DampingAlpha=1,MaxLinearVelocity=10000,MaxAngularVelocity=10000,MaxLinearAcceleration=10000,MaxAngularAcceleration=10000,ExternalLinearDragV=(X=0,Y=0,Z=0),ExternalLinearVelocity=(X=0,Y=0,Z=0),ExternalAngularVelocity=(X=0,Y=0,Z=0))
   ) {
    bOverrideWorldGravity=true
    overridePhysicsAsset=&/F_MED_OutlineLove_Blur_Cape_physics
    simulationSpace=BaseBoneSpace
    componentAppliedLinearAccClamp=<3000 3000 3000>
   }
   Save cached pose 'RbanCache'(Pose=Blend Poses by bool(
     BlendPose_0=Component To Local(ComponentPose=@n44)
     BlendPose_1=Component To Local(ComponentPose=@n45)
     BlendTime_0=0.2
     BlendTime_1=0.5
     bActiveValue=Get bClothEnabled()
    ) {blendType=Linear}) {cacheName=RbanCache}
   n46 = Layered blend per bone(
    BasePose=Copy Pose From Mesh() {bUseAttachedParent=true}
    BlendPoses_0=Copy Pose From Mesh(SourceMeshComponent=Get hair_mesh() -> (hair_mesh:Skeletal Mesh Component Object Reference)) {bUseAttachedParent=true}
   ) {layerSetup [{branchFilters [
       {blendDepth=1, boneName=dyn_hair_bk_MSTR}
       {blendDepth=1, boneName=dyn_hair_bk_00}
       {blendDepth=1, boneName=dyn_hair_bk_01}
       {blendDepth=1, boneName=dyn_hair_bk_02}
       {blendDepth=1, boneName=dyn_hair_bk_03}
       {blendDepth=1, boneName=dyn_hair_bk_04}
      ]}]}
   n47 = "Constraint - dyn_skirt_l_1_01 (P)"(ComponentPose=Local To Component(LocalPose=@n46), Alpha=1, ConstraintWeights_0=1) {boneToModify {boneName=dyn_skirt_l_1_01}, constraintSetup [{targetBone {boneName=thigh_l}}], constraintWeights [1]}
   n48 = "Constraint - dyn_skirt_l_2_01 (PP)"!(ComponentPose=@n47, Alpha=1, ConstraintWeights_0=0.7, ConstraintWeights_1:float=0.3) -> (Pose) {
    boneToModify {boneName=dyn_skirt_l_2_01}
    constraintSetup [{targetBone {boneName=thigh_l}}, {targetBone {boneName=base_skirt_l_3_01}}]
    constraintWeights [1,1]
   }
   n49 = "Constraint - dyn_skirt_r_1_01 (P)"(ComponentPose=@n48, Alpha=1, ConstraintWeights_0=1) {boneToModify {boneName=dyn_skirt_r_1_01}, constraintSetup [{targetBone {boneName=thigh_r}}], constraintWeights [1]}
   n50 = "Constraint - dyn_skirt_r_2_01 (PP)"!(ComponentPose=@n49, Alpha=1, ConstraintWeights_0=0.7, ConstraintWeights_1:float=0.3) -> (Pose) {
    boneToModify {boneName=dyn_skirt_r_2_01}
    constraintSetup [{targetBone {boneName=thigh_r}}, {targetBone {boneName=base_skirt_r_3_01}}]
    constraintWeights [1,1]
   }
   n51 = Anim Dynamics - Bone: dyn_skirt_l_1_01(ComponentPose=@n50, Alpha=1, ExternalForce="0,0,0") {
    angularDampingOverride=0.99
    angularSpringConstant=800
    bAngularSpring=true
    bOverrideAngularDamping=true
    bUseGravityOverride=true
    gravityScale=0.3
    simulationSpace=BoneRelative
    alphaBoolBlend {blendOption=Linear}
    boundBone {boneName=dyn_skirt_l_1_01}
    componentAppliedLinearAccClamp=<2500 2500 1500>
    componentLinearAccScale=<0.5 0.5 0.4>
    componentLinearVelScale=<0.55 0.55 0.45>
    gravityOverride {z=-980}
    physicsBodyDefinitions [{
      collisionType=CustomSphere
      sphereCollisionRadius=2
      boundBone {boneName=dyn_skirt_l_1_01}
      boxExtents=<60 60 60>
      constraintSetup {angularLimitsMax=<0 0 0>, angularLimitsMin=<0 0 -20>, angularTarget=<0 0 0>}
      localJointOffset=<5 0 0>
     }]
    relativeSpaceBone {boneName=root}
   }
   n52 = Anim Dynamics - Bone: dyn_skirt_l_2_01(ComponentPose=@n51, Alpha=1, ExternalForce="0,0,0") {
    angularDampingOverride=0.99
    angularSpringConstant=800
    bAngularSpring=true
    bOverrideAngularDamping=true
    bUseGravityOverride=true
    gravityScale=0.3
    simulationSpace=BoneRelative
    alphaBoolBlend {blendOption=Linear}
    boundBone {boneName=dyn_skirt_l_2_01}
    componentAppliedLinearAccClamp=<2500 2500 1500>
    componentLinearAccScale=<0.5 0.5 0.4>
    componentLinearVelScale=<0.55 0.55 0.45>
    gravityOverride {z=-980}
    physicsBodyDefinitions [{
      collisionType=CustomSphere
      sphereCollisionRadius=2
      boundBone {boneName=dyn_skirt_l_2_01}
      boxExtents=<60 60 60>
      constraintSetup {angularLimitsMax=<0 10 0>, angularLimitsMin=<0 -10 -20>, angularTarget=<0 0 0>}
      localJointOffset=<5 0 0>
     }]
    relativeSpaceBone {boneName=root}
   }
   n53 = Anim Dynamics - Bone: dyn_skirt_l_1_02(ComponentPose=@n52, Alpha=1, ExternalForce="0,0,0") {
    angularDampingOverride=0.99
    angularSpringConstant=800
    bAngularSpring=true
    bOverrideAngularDamping=true
    bUseGravityOverride=true
    gravityScale=0.3
    simulationSpace=BoneRelative
    alphaBoolBlend {blendOption=Linear}
    boundBone {boneName=dyn_skirt_l_1_02}
    componentAppliedLinearAccClamp=<2500 2500 1500>
    componentLinearAccScale=<0.5 0.5 0.4>
    componentLinearVelScale=<0.55 0.55 0.45>
    gravityOverride {z=-980}
    physicsBodyDefinitions [{
      collisionType=CustomSphere
      sphereCollisionRadius=2
      boundBone {boneName=dyn_skirt_l_1_02}
      boxExtents=<40 40 40>
      constraintSetup {angularLimitsMax=<0 0 0>, angularLimitsMin=<0 0 -20>, angularTarget=<0 0 0>}
      localJointOffset=<5 0 0>
     }]
    relativeSpaceBone {boneName=root}
   }
   n54 = Anim Dynamics - Bone: dyn_skirt_l_2_02(ComponentPose=@n53, Alpha=1, ExternalForce="0,0,0") {
    angularDampingOverride=0.99
    angularSpringConstant=800
    bAngularSpring=true
    bOverrideAngularDamping=true
    bUseGravityOverride=true
    gravityScale=0.3
    simulationSpace=BoneRelative
    alphaBoolBlend {blendOption=Linear}
    boundBone {boneName=dyn_skirt_l_2_02}
    componentAppliedLinearAccClamp=<2500 2500 1500>
    componentLinearAccScale=<0.5 0.5 0.4>
    componentLinearVelScale=<0.55 0.55 0.45>
    gravityOverride {z=-980}
    physicsBodyDefinitions [{
      collisionType=CustomSphere
      sphereCollisionRadius=2
      boundBone {boneName=dyn_skirt_l_2_02}
      boxExtents=<40 40 40>
      constraintSetup {angularLimitsMax=<0 10 0>, angularLimitsMin=<0 -10 -20>, angularTarget=<0 0 0>}
      localJointOffset=<5 0 0>
     }]
    relativeSpaceBone {boneName=root}
    sphericalLimits [{limitRadius=22, drivingBone {boneName=lowerarm_l}, sphereLocalOffset=<25 0 17>}]
   }
   n55 = Anim Dynamics - Bone: dyn_skirt_l_1_03(ComponentPose=@n54, Alpha=1, ExternalForce="0,0,0") {
    angularDampingOverride=0.99
    angularSpringConstant=800
    bAngularSpring=true
    bOverrideAngularDamping=true
    bUseGravityOverride=true
    gravityScale=0.3
    simulationSpace=BoneRelative
    alphaBoolBlend {blendOption=Linear}
    boundBone {boneName=dyn_skirt_l_1_03}
    componentAppliedLinearAccClamp=<2500 2500 1500>
    componentLinearAccScale=<0.5 0.5 0.4>
    componentLinearVelScale=<0.55 0.55 0.45>
    gravityOverride {z=-980}
    physicsBodyDefinitions [{
      collisionType=CustomSphere
      sphereCollisionRadius=2
      boundBone {boneName=dyn_skirt_l_1_03}
      boxExtents=<20 20 20>
      constraintSetup {angularLimitsMax=<0 0 0>, angularLimitsMin=<0 0 -20>, angularTarget=<0 0 0>}
      localJointOffset=<5 0 0>
     }]
    relativeSpaceBone {boneName=root}
   }
   n56 = Anim Dynamics - Bone: dyn_skirt_l_2_03(ComponentPose=@n55, Alpha=1, ExternalForce="0,0,0") {
    angularDampingOverride=0.99
    angularSpringConstant=800
    bAngularSpring=true
    bOverrideAngularDamping=true
    bUseGravityOverride=true
    bUseSphericalLimits=true
    gravityScale=0.3
    simulationSpace=BoneRelative
    alphaBoolBlend {blendOption=Linear}
    boundBone {boneName=dyn_skirt_l_2_03}
    componentAppliedLinearAccClamp=<2500 2500 1500>
    componentLinearAccScale=<0.5 0.5 0.4>
    componentLinearVelScale=<0.55 0.55 0.45>
    gravityOverride {z=-980}
    physicsBodyDefinitions [{
      collisionType=CustomSphere
      sphereCollisionRadius=2
      boundBone {boneName=dyn_skirt_l_2_03}
      boxExtents=<20 20 20>
      constraintSetup {angularLimitsMax=<0 10 0>, angularLimitsMin=<0 -10 -20>, angularTarget=<0 0 0>}
      localJointOffset=<5 0 0>
     }]
    relativeSpaceBone {boneName=root}
    sphericalLimits [{limitRadius=22, drivingBone {boneName=lowerarm_l}, sphereLocalOffset=<25 0 17>}]
   }
   n57 = Anim Dynamics - Bone: dyn_skirt_r_1_01(ComponentPose=@n56, Alpha=1, ExternalForce="0,0,0") {
    angularDampingOverride=0.99
    angularSpringConstant=800
    bAngularSpring=true
    bOverrideAngularDamping=true
    bUseGravityOverride=true
    gravityScale=0.3
    simulationSpace=BoneRelative
    alphaBoolBlend {blendOption=Linear}
    boundBone {boneName=dyn_skirt_r_1_01}
    componentAppliedLinearAccClamp=<2500 2500 1500>
    componentLinearAccScale=<0.5 0.5 0.4>
    componentLinearVelScale=<0.55 0.55 0.45>
    gravityOverride {z=-980}
    physicsBodyDefinitions [{
      collisionType=CustomSphere
      sphereCollisionRadius=2
      boundBone {boneName=dyn_skirt_r_1_01}
      boxExtents=<60 60 60>
      constraintSetup {angularLimitsMax=<1 1 0>, angularLimitsMin=<-1 -1 -20>, angularTarget=<0 0 0>}
      localJointOffset=<5 0 0>
     }]
    relativeSpaceBone {boneName=root}
   }
   n58 = Anim Dynamics - Bone: dyn_skirt_r_2_01(ComponentPose=@n57, Alpha=1, ExternalForce="0,0,0") {
    angularDampingOverride=0.99
    angularSpringConstant=800
    bAngularSpring=true
    bOverrideAngularDamping=true
    bUseGravityOverride=true
    gravityScale=0.3
    simulationSpace=BoneRelative
    alphaBoolBlend {blendOption=Linear}
    boundBone {boneName=dyn_skirt_r_2_01}
    componentAppliedLinearAccClamp=<2500 2500 1500>
    componentLinearAccScale=<0.5 0.5 0.4>
    componentLinearVelScale=<0.55 0.55 0.45>
    gravityOverride {z=-980}
    physicsBodyDefinitions [{
      collisionType=CustomSphere
      sphereCollisionRadius=2
      boundBone {boneName=dyn_skirt_r_2_01}
      boxExtents=<60 60 60>
      constraintSetup {angularLimitsMax=<0 5 0>, angularLimitsMin=<0 -5 -20>, angularTarget=<0 0 0>}
      localJointOffset=<5 0 0>
     }]
    relativeSpaceBone {boneName=root}
   }
   n59 = Anim Dynamics - Bone: dyn_skirt_r_1_02(ComponentPose=@n58, Alpha=1, ExternalForce="0,0,0") {
    angularDampingOverride=0.99
    angularSpringConstant=800
    bAngularSpring=true
    bOverrideAngularDamping=true
    bUseGravityOverride=true
    gravityScale=0.3
    simulationSpace=BoneRelative
    alphaBoolBlend {blendOption=Linear}
    boundBone {boneName=dyn_skirt_r_1_02}
    componentAppliedLinearAccClamp=<2500 2500 1500>
    componentLinearAccScale=<0.5 0.5 0.4>
    componentLinearVelScale=<0.55 0.55 0.45>
    gravityOverride {z=-980}
    physicsBodyDefinitions [{
      collisionType=CustomSphere
      sphereCollisionRadius=2
      boundBone {boneName=dyn_skirt_r_1_02}
      boxExtents=<40 40 40>
      constraintSetup {angularLimitsMax=<1 1 0>, angularLimitsMin=<-1 -1 -20>, angularTarget=<0 0 0>}
      localJointOffset=<5 0 0>
     }]
    relativeSpaceBone {boneName=root}
   }
   n60 = Anim Dynamics - Bone: dyn_skirt_r_2_02(ComponentPose=@n59, Alpha=1, ExternalForce="0,0,0") {
    angularDampingOverride=0.99
    angularSpringConstant=800
    bAngularSpring=true
    bOverrideAngularDamping=true
    bUseGravityOverride=true
    gravityScale=0.3
    simulationSpace=BoneRelative
    alphaBoolBlend {blendOption=Linear}
    boundBone {boneName=dyn_skirt_r_2_02}
    componentAppliedLinearAccClamp=<2500 2500 1500>
    componentLinearAccScale=<0.5 0.5 0.4>
    componentLinearVelScale=<0.55 0.55 0.45>
    gravityOverride {z=-980}
    physicsBodyDefinitions [{
      collisionType=CustomSphere
      sphereCollisionRadius=2
      boundBone {boneName=dyn_skirt_r_2_02}
      boxExtents=<40 40 40>
      constraintSetup {angularLimitsMax=<0 5 0>, angularLimitsMin=<0 -5 -20>, angularTarget=<0 0 0>}
      localJointOffset=<5 0 0>
     }]
    relativeSpaceBone {boneName=root}
   }
   n61 = Anim Dynamics - Bone: dyn_skirt_r_1_03(ComponentPose=@n60, Alpha=1, ExternalForce="0,0,0") {
    angularDampingOverride=0.99
    angularSpringConstant=800
    bAngularSpring=true
    bOverrideAngularDamping=true
    bUseGravityOverride=true
    gravityScale=0.3
    simulationSpace=BoneRelative
    alphaBoolBlend {blendOption=Linear}
    boundBone {boneName=dyn_skirt_r_1_03}
    componentAppliedLinearAccClamp=<2500 2500 1500>
    componentLinearAccScale=<0.5 0.5 0.4>
    componentLinearVelScale=<0.55 0.55 0.45>
    gravityOverride {z=-980}
    physicsBodyDefinitions [{
      collisionType=CustomSphere
      sphereCollisionRadius=2
      boundBone {boneName=dyn_skirt_r_1_03}
      boxExtents=<20 20 20>
      constraintSetup {angularLimitsMax=<1 1 0>, angularLimitsMin=<-1 -1 -20>, angularTarget=<0 0 0>}
      localJointOffset=<5 0 0>
     }]
    relativeSpaceBone {boneName=root}
   }
   n62 = Anim Dynamics - Bone: dyn_skirt_r_2_03(ComponentPose=@n61, Alpha=1, ExternalForce="0,0,0") {
    angularDampingOverride=0.99
    angularSpringConstant=800
    bAngularSpring=true
    bOverrideAngularDamping=true
    bUseGravityOverride=true
    bUseSphericalLimits=true
    gravityScale=0.3
    simulationSpace=BoneRelative
    alphaBoolBlend {blendOption=Linear}
    boundBone {boneName=dyn_skirt_r_2_03}
    componentAppliedLinearAccClamp=<2500 2500 1500>
    componentLinearAccScale=<0.5 0.5 0.4>
    componentLinearVelScale=<0.55 0.55 0.45>
    gravityOverride {z=-980}
    physicsBodyDefinitions [{
      collisionType=CustomSphere
      sphereCollisionRadius=2
      boundBone {boneName=dyn_skirt_r_2_03}
      boxExtents=<20 20 20>
      constraintSetup {angularLimitsMax=<0 5 0>, angularLimitsMin=<0 -5 -20>, angularTarget=<0 0 0>}
      localJointOffset=<5 0 0>
     }]
    relativeSpaceBone {boneName=root}
    sphericalLimits [
     {limitRadius=22, drivingBone {boneName=lowerarm_r}, sphereLocalOffset=<-25 0 -17>}
     {limitRadius=0, drivingBone {boneName=None}, sphereLocalOffset=<0 0 0>}
    ]
   }
   n63 = RigidBody!(ComponentPose=@n62, Alpha, OverrideWorldGravity="0,0,-980", ExternalForce) -> (Pose) {
    bClampLinearTranslationLimitToRefPose=true
    bOverrideWorldGravity=true
    overridePhysicsAsset=&/F_MED_OutlineLove_Blur_physics
    simulationSpace=BaseBoneSpace
    baseBoneRef {boneName=pelvis}
    simSpaceSettings {dampingAlpha=0.25, maxAngularAcceleration=8, maxAngularVelocity=2.5, maxLinearAcceleration=800, maxLinearVelocity=300, worldAlpha=1}
   }
   Save cached pose 'SkirtDynPose'(Pose=Component To Local(ComponentPose=@n63)) {cacheName=SkirtDynPose}
  ]}
 EventGraph {nodes [
   n1 = Event BlueprintUpdateAnimation()
   > Set DeltaTimeLocal(DeltaTimeLocal:double=@n1.DeltaTimeX) -> (Output_Get:double)
   > n2 = Sequence!(execute) -> (then_0, then_1, then_2:Exec, then_3:Exec, then_4:Exec, then_5:Exec, then_6:Exec, then_7:Exec, then_8:Exec)
   > n3 = Branch(Condition=Get bClothEnabled())
   > n4 = "Cloth Controller Function (Parameter Asset) (Nativized)"!(
    execute
    CharacterPartAnimInstance:Custom Character Part Anim Instance Object Reference=Self-Reference()
    ClothingAssetName:Name=F_MED_OutlineLove_Blur_Clothing
    ClothCoreSettings:"Cloth Core Settings Structure (by ref)"=Get Cloth Core Settings_Native() -> (Cloth Core Settings_Native:Cloth Core Settings Structure)
    WindGustLocalSettings:"Wind Gust Settings Structure (by ref)"=Get Wind Gust Local Settings Native() -> (Wind Gust Local Settings Native:Wind Gust Settings Structure)
    ClothUpAxis:Vector="1,0,0"
    LinVelScaleFwdBwdClamp:float=0
    LinVelScaleUpDwnClamp:float=0.7
    DeltaTime:float=Get DeltaTimeLocal() -> (DeltaTimeLocal:double)
    GameTimeInSeconds:float=GetGameTimeInSeconds()
    StateLogicLayerName:Name=none
    Parameters:Fort Character Dynamics Parameters Object Reference=&/M_MED_OutlineLove_Blur_Cape_Parameters
   ) -> (
    then
    OutCurrentState:Name=None
    OutClothParameters:Cloth Parameters Structure
    OutClothCoreSettings:Cloth Core Settings Structure
    OutWindGustSettings:Wind Gust Parameters Structure
    OutWindGustLocalSettings:Wind Gust Settings Structure
    OutWindVelocity:Vector
    OutLinearVelVector:Vector
   )
   > Set Wind Gust Local Settings Out(Wind Gust Local Settings Out:Wind Gust Settings Structure=@n4.OutWindGustLocalSettings) -> (Output_Get:Wind Gust Settings Structure)
   > n5 = Branch(Condition=Get isWearingPack())
   n6 = GetPartSkeletalMeshComponent(self=Self-Reference())
   > Set ClothBlendWeight!(execute, ClothBlendWeight:float=0, self:Skeletal Mesh Component Object Reference=@n6) -> (then, Output_Get:float)
   > Set bDisableClothSimulation!(execute, bDisableClothSimulation:bool=true, self:Skeletal Mesh Component Object Reference=@n6) -> (then, Output_Get:bool)
   Set ClothBlendWeight(execute=@n5.else, ClothBlendWeight=1, self=@n6)
   > Set bDisableClothSimulation(bDisableClothSimulation=false, self=@n6)
   n7 = "Gravity Override State Function (Parameter Asset) (Nativized)"(
    execute=@n3.else
    CharacterPartAnimInstance=Self-Reference()
    GameTimeInSeconds=GetGameTimeInSeconds()
    Parameters=&M_MED_OutlineLove_Blur_Cape_Parameters
   )
   > Set GravityOverride(GravityOverride=@n7)
   > Set RBAN Settings Out(RBAN Settings Out=@n7.OutRigidBodyAnimNodeParameters)
   n8 = PlayerMovement(execute=@n2.then_1, AnimInstance=Self-Reference(), XDotProductCoefficient=1, YDotProductCoefficient=1, ZDotProductCoefficient=1)
   n9 = MapRangeClamped(Value=@n8.YMovement, InRangeA=0, InRangeB=1)
   n10 = "Absolute (Float)"(A=@n8)
   > Set BwdMovement(BwdMovement:double=MapRangeClamped(Value=float + float(A=@n10, B=@n9) {is=title K2Node_CommutativeAssociativeBinaryOperator}, InRangeA=0, InRangeB=1.4)) -> (Output_Get:double)
   n11 = Branch(execute=@n2.then_3, Condition=Get bIsInFrontEnd())
   > Set BwdMovement(BwdMovement=0.25)
   n12 = Branch(execute=@n11.else, Condition=Get bIsSwimSprinting() -> (bIsSwimSprinting:bool))
   > Set TrailControllerRot(TrailControllerRot:Rotator="15,0, 0") -> (Output_Get:Rotator)
   n13 = Branch(execute=@n12.else, Condition=Get bIsSkyDiving())
   > n14 = Branch(Condition=Get bIsParachuteOpened())
   > Set TrailControllerRot(TrailControllerRot="45,0, 0")
   n15 = Branch(execute=@n14.else, Condition="float > float"(A=Get PawnVelocityZ(), B=50) {is=title K2Node_CallFunction})
   > Set BwdMovement(BwdMovement=float * float(A=Get BwdMovement(), B=0))
   > Set TrailControllerRot(TrailControllerRot="45,0, 0")
   n16 = Branch(execute=@n15.else, Condition=Get bIsSkydiveDiveMode())
   > Set TrailControllerRot(TrailControllerRot="50,0,0")
   Set TrailControllerRot(execute=@n16.else, TrailControllerRot="60,0,0")
   Set TrailControllerRot(execute=@n13.else, TrailControllerRot="5,0,0")
   > Set BwdMovement(BwdMovement=float * float(A=Get BwdMovement(), B=1))
   n17 = PlayerMovement(execute=@n2.then_4, AnimInstance=Self-Reference(), XDotProductCoefficient=1, YDotProductCoefficient=1, ZDotProductCoefficient=1)
   n18 = BreakVector(InVec=MakeVector(X=@n17, Y=@n17.YMovement))
   > Set strafeLeft float(strafeLeft float:double="Absolute (Float)"(A="Clamp (Float)"(Value=@n18, Min=-1, Max=0))) -> (Output_Get:double)
   > Set strafeRight float(strafeRight float:double="Clamp (Float)"(Value=@n18, Min=0, Max=1)) -> (Output_Get:double)
   > Set moveForward(moveForward:double="Clamp (Float)"(Value=@n18.Y, Min=0, Max=1)) -> (Output_Get:double)
   > Set moveBackward(moveBackward:double="Absolute (Float)"(A="Clamp (Float)"(Value=@n18.Y, Min=-1, Max=0))) -> (Output_Get:double)
   n19 = NOT Boolean(A=Get bIsMoving2D())
   n20 = Get bIsJumping()
   Set JumpingNotMoving2D(execute=@n2.then_5, JumpingNotMoving2D:double=SelectFloat(A=1, B=0, bPickA=AND Boolean(A=@n20, B=@n19))) -> (Output_Get:double)
   n21 = Cast To PlayerPawn_Athena(execute=@n2.then_6, Object=TryGetPawnOwner())
   > Set hair_mesh(hair_mesh:Skeletal Mesh Component Object Reference=GetSkeletalMeshForPartType(self=@n21, PartType=Face)) -> (Output_Get:Skeletal Mesh Component Object Reference)
   n22 = Get isKneeSlidingB() -> (isKneeSlidingB:bool)
   Set isKneeSlidingF(
    execute=@n2.then_8
    isKneeSlidingF:double=FInterpTo(
     Current=Get isKneeSlidingF() -> (isKneeSlidingF:double)
     Target="To Float (Boolean)"(InBool=@n22)
     DeltaTime=Get DeltaTimeLocal()
     InterpSpeed=10
    )
   ) -> (Output_Get:double)
   n23 = Get isbaseBallSlidingB() -> (isbaseBallSlidingB:bool)
   > Set isBaseBallSlidingF(isBaseBallSlidingF:double=FInterpTo(
     Current=Get isBaseBallSlidingF() -> (isBaseBallSlidingF:double)
     Target="To Float (Boolean)"(InBool=@n23)
     DeltaTime=Get DeltaTimeLocal()
     InterpSpeed=10
    )) -> (Output_Get:double)
   n24 = Branch(execute=@n2.then_7, Condition=Get bIsSliding())
   n25 = GetAlphaFromReferenceTransformDelta(SkeletalMeshComponent=GetPartSkeletalMeshComponent(), BoneName=calf_l, BaseName=thigh_l, InputRange=(X=0,Y=1), TwistAxis="1,0,0")
   > n26 = Branch(Condition=NOT Boolean(A="float > float"(A=@n25.RawOutput, B=100) {is=title K2Node_PromotableOperator}))
   > Set isKneeSlidingB(isKneeSlidingB:bool=false) -> (Output_Get:bool)
   > Set isbaseBallSlidingB(isbaseBallSlidingB:bool=true) -> (Output_Get:bool)
   Set isKneeSlidingB(execute=@n26.else, isKneeSlidingB=true)
   > Set isbaseBallSlidingB(isbaseBallSlidingB=false)
   Set isKneeSlidingB(execute=@n24.else, isKneeSlidingB=false)
   > Set isbaseBallSlidingB(isbaseBallSlidingB=false)
   n27 = Branch(execute=@n2.then_2, Condition=OR Boolean!(A=Get bIsSkyDiving(), B=Get bIsCrouching(), C:bool=Get bIsPlayingEmote()) -> (ReturnValue))
   > Set isIdle(isIdle:double=0) -> (Output_Get:double)
   > Set isIdleInFE(isIdleInFE:double=0) -> (Output_Get:double)
   n28 = float + float!(
    A=Get strafeLeft float() -> (strafeLeft float:double)
    B=Get strafeRight float() -> (strafeRight float:double)
    C:double=Get moveForward() -> (moveForward:double)
    D:double=Get moveBackward() -> (moveBackward:double)
   ) -> (ReturnValue) {is=title K2Node_PromotableOperator}
   Set isIdle(execute=@n27.else, isIdle=Select(Option 0:double=1, Option 1:double=0, Index="float > float"(A=@n28)) -> (ReturnValue:double))
   > n29 = Branch(Condition=Get bIsInFrontEnd())
   > Set isIdleInFE(isIdleInFE=Get isIdle() -> (isIdle:double))
   Set isIdleInFE(execute=@n29.else, isIdleInFE=0)
   Event OnCharacterCustomizationLoaded()
   n30 = CheckForBackbling()
   > Set isWearingPack(isWearingPack:bool=OR Boolean(A=@n30, B=@n30.bIsBackblingACape, C=@n30.bIsBackblingLong)) -> (Output_Get:bool)
   > Set isWearingCape(isWearingCape=@n30.bIsBackblingLong)
   Event OnEnterGrindRail() {is=title K2Node_Event}
   > Set isOnGrindRailF(isOnGrindRailF:double=1) -> (Output_Get:double)
   Event OnExitGrindRail() {is=title K2Node_Event}
   > Set isOnGrindRailF(isOnGrindRailF=0)
   Event OnBeginCrouching()
   > Set isCrouchingF(isCrouchingF:double=1) -> (Output_Get:double)
   Event OnEndCrouching()
   > Set isCrouchingF(isCrouchingF=0)
   Event OnBeginSkydiving()
   > Set TrailControllerRot(TrailControllerRot="-65,0,0")
   > Set isSkyDivingF(isSkyDivingF:double=1) -> (Output_Get:double)
   Event OnEndSkydiving()
   > Set TrailControllerRot(TrailControllerRot="-10,0,0")
   > Set isSkyDivingF(isSkyDivingF=0)
  ]}
]
variables [
 Anim Dynamics Cape {type=boolean}
 Backbling Rotation {type=Rotator}
 Backbling Translation {type=Vector}
 BwdMovement {type=number}
 Cape_LookAt_Target {type=Vector}
 ChaosClothingInteractor {type=ChaosClothingInteractor, default=null}
 Cloth Core Settings_Native {
  type=ClothCoreSettings
  default {
   collisionThickness=0
   dampingCoefficient=0.05
   frictionCoefficient=0.3
   localDampingCoefficient=0
   numberIterations=2
   numberSubsteps=1
   selfCollisionThickness=0
   animDriveDamping=<0 1>
   animDriveStiffness=<0 1>
   areaStiffness=<1 1>
   bendingStiffness=<1 1>
   edgeStiffness=<1 1>
   tetherStiffness=<1 1>
  }
 }
 DeltaTimeLocal {type=number}
 GravityOverride {type=Vector}
 hair_mesh {type=SkeletalMeshComponent}
 isbaseBallSlidingB {type=boolean}
 isBaseBallSlidingF {type=number}
 isCrouchingF {type=number}
 isIdle {type=number}
 isIdleInFE {type=number}
 isKneeSlidingB {type=boolean}
 isKneeSlidingF {type=number}
 isOnGrindRailF {type=number}
 isSkyDivingF {type=number}
 isWearingCape {type=boolean}
 isWearingPack {type=boolean}
 JumpingNotMoving2D {type=number}
 moveBackward {type=number}
 moveForward {type=number}
 RBAN Settings Out {type=FortRigidBodyAnimNodeParameters}
 RbanOrCloth {type=boolean}
 SkydivingAlpha {type=Rotator}
 strafeLeft float {type=number}
 strafeRight float {type=number}
 TrailControllerRot {type=Rotator}
 Wind Gust Local Settings Native {type=WindGustSettings, default {elapsedTime=0, finalGustStrengthLocal=0, gustDuration=0, gustStrength=0, timeBetweenGusts=0}}
 Wind Gust Local Settings Out {type=WindGustSettings, default {elapsedTime=0, finalGustStrengthLocal=0, gustDuration=0, gustStrength=0, timeBetweenGusts=0}}
]
