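# Anim Blueprint for the F_MED_OutlineLove_Blur face accessory, animating the character's dynamic hair over the copied body pose.
# Backbling selects between two RigidBody setups feeding a front-physics pass, while bRigidBodyEnabled switches between that and a constraint fallback following cape bones.
# AnimDynamics front strands, front-end idle and emote corrections, then a Control Rig finish the pose.
# The EventGraph derives gravity and backbling, sprint, skydive, emote and idle alphas.
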AnimBlueprint /BRCosmetics/Characters/Player/Female/Medium/Bodies/F_MED_OutlineLove_Blur/Meshes/Parts/F_MED_OutlineLove_Blur_FaceAcc_AnimBP
target_skeleton=&/F_MED_OutlineLove_Blur_FaceAcc_Skeleton
graphs [
 AnimGraph {nodes [
   n1 = RigidBody!(
    ComponentPose=Local To Component(LocalPose=Copy Pose From Mesh() {bUseAttachedParent=true})
    bAlphaBoolEnabled:bool
    OverrideWorldGravity:Vector="0,0,0"
    ExternalForce
    ComponentLinearAccScale:Vector="0,0,0"
    ComponentLinearVelScale:Vector="0,0,0"
   ) -> (Pose) {
    alphaInputType=Bool
    bClampLinearTranslationLimitToRefPose=true
    bOverrideWorldGravity=true
    overridePhysicsAsset=&/F_MED_OutlineLove_Blur_FaceAcc_Physics
    simulationSpace=BaseBoneSpace
    alphaBoolBlend {blendOption=Linear}
    alphaScaleBiasClamp {interpSpeedDecreasing=4, interpSpeedIncreasing=6}
    baseBoneRef {boneName=spine_05}
    componentAppliedLinearAccClamp=<3500 3500 3500>
    componentLinearAccScale=<0.6 0.6 0.2>
    componentLinearVelScale=<0.5 0.5 0.5>
    overrideWorldGravity {z=-875}
    simSpaceSettings {
     dampingAlpha=0
     maxAngularAcceleration=40
     maxAngularVelocity=8
     maxLinearAcceleration=1500
     maxLinearVelocity=700
     velocityScaleZ=0.5
     worldAlpha=0.7
    }
   }
   n2 = "Transform (Modify) Bone - Bone: dyn_hair_bk_02"(ComponentPose=@n1, Alpha=1, Translation="0,0,0", Rotation="0,0,20", Scale="1,1,1") {rotationMode=BMM_Additive, translationMode=BMM_Additive, boneToModify {boneName=dyn_hair_bk_02}, rotation {roll=20}}
   n3 = "Transform (Modify) Bone - Bone: dyn_hair_bk_04"(
    ComponentPose="Transform (Modify) Bone - Bone: dyn_hair_l_02"(
     ComponentPose="Transform (Modify) Bone - Bone: dyn_hair_r_02"(ComponentPose=@n2, Alpha=1, Translation="0,0,0", Rotation="0,0,20", Scale="1,1,1") {rotationMode=BMM_Additive, translationMode=BMM_Additive, boneToModify {boneName=dyn_hair_r_02}, rotation {roll=20}}
     Alpha=1
     Translation="0,0,0"
     Rotation="0,0,20"
     Scale="1,1,1"
    ) {rotationMode=BMM_Additive, translationMode=BMM_Additive, boneToModify {boneName=dyn_hair_l_02}, rotation {roll=20}}
    Alpha=1
    Translation="0,1,0"
    Rotation="0,0,20"
    Scale="1,1,1"
   ) {rotationMode=BMM_Additive, translationMode=BMM_Additive, boneToModify {boneName=dyn_hair_bk_04}, rotation {roll=20}, translation {y=1}}
   n4 = "Transform (Modify) Bone - Bone: dyn_hair_l_02"(ComponentPose=@n3, Alpha=1, Translation="-1.5,0,1.5", Rotation="10,0,0", Scale="1,1,1") {
    rotationMode=BMM_Additive
    rotationSpace=BCS_BoneSpace
    translationMode=BMM_Additive
    translationSpace=BCS_BoneSpace
    alphaBoolBlend {blendOption=Linear}
    alphaScaleBiasClamp {bInterpResult=true, interpSpeedDecreasing=2, interpSpeedIncreasing=3}
    boneToModify {boneName=dyn_hair_l_02}
    rotation {pitch=10}
    translation {x=-1.5, z=1.5}
   }
   n5 = "Transform (Modify) Bone - Bone: dyn_hair_bk_01"(ComponentPose=@n4, Alpha=1, Translation="-1.5,0,1.5", Rotation="10,0,0", Scale="1,1,1") {
    rotationMode=BMM_Additive
    rotationSpace=BCS_BoneSpace
    translationMode=BMM_Additive
    translationSpace=BCS_BoneSpace
    alphaBoolBlend {blendOption=Linear}
    alphaScaleBiasClamp {bInterpResult=true, interpSpeedDecreasing=2, interpSpeedIncreasing=3}
    boneToModify {boneName=dyn_hair_bk_01}
    rotation {pitch=10}
    translation {x=-1.5, z=1.5}
   }
   n6 = "Transform (Modify) Bone - Bone: dyn_hair_r_02"(ComponentPose=@n5, Alpha=1, Translation="-1.5,0,1.5", Rotation="10,0,0", Scale="1,1,1") {
    rotationMode=BMM_Additive
    rotationSpace=BCS_BoneSpace
    translationMode=BMM_Additive
    translationSpace=BCS_BoneSpace
    alphaBoolBlend {blendOption=Linear}
    alphaScaleBiasClamp {bInterpResult=true, interpSpeedDecreasing=2, interpSpeedIncreasing=3}
    boneToModify {boneName=dyn_hair_r_02}
    rotation {pitch=10}
    translation {x=-1.5, z=1.5}
   }
   n7 = "Transform (Modify) Bone - Bone: dyn_hair_l_03"(ComponentPose=@n6, Alpha=1, Translation="0,0,0", Rotation="5,0,0", Scale="1,1,1") {
    rotationMode=BMM_Additive
    rotationSpace=BCS_BoneSpace
    translationMode=BMM_Additive
    translationSpace=BCS_BoneSpace
    alphaBoolBlend {blendOption=Linear}
    alphaScaleBiasClamp {bInterpResult=true, interpSpeedDecreasing=2, interpSpeedIncreasing=3}
    boneToModify {boneName=dyn_hair_l_03}
    rotation {pitch=5}
   }
   n8 = "Transform (Modify) Bone - Bone: dyn_hair_bk_03"(ComponentPose=@n7, Alpha=1, Translation="0,0,0", Rotation="10,0,0", Scale="1,1,1") {
    rotationMode=BMM_Additive
    rotationSpace=BCS_BoneSpace
    translationMode=BMM_Additive
    translationSpace=BCS_BoneSpace
    alphaBoolBlend {blendOption=Linear}
    alphaScaleBiasClamp {bInterpResult=true, interpSpeedDecreasing=2, interpSpeedIncreasing=3}
    boneToModify {boneName=dyn_hair_bk_03}
    rotation {pitch=10}
   }
   n9 = "Transform (Modify) Bone - Bone: dyn_hair_r_03"(ComponentPose=@n8, Alpha=1, Translation="0,0,0", Rotation="5,0,0", Scale="1,1,1") {
    rotationMode=BMM_Additive
    rotationSpace=BCS_BoneSpace
    translationMode=BMM_Additive
    translationSpace=BCS_BoneSpace
    alphaBoolBlend {blendOption=Linear}
    alphaScaleBiasClamp {bInterpResult=true, interpSpeedDecreasing=2, interpSpeedIncreasing=3}
    boneToModify {boneName=dyn_hair_r_03}
    rotation {pitch=5}
   }
   n10 = "Transform (Modify) Bone - Bone: dyn_hair_l_04"(ComponentPose=@n9, Alpha=1, Translation="0,0,0", Rotation="25,0,0", Scale="1,1,1") {
    rotationMode=BMM_Additive
    rotationSpace=BCS_BoneSpace
    translationMode=BMM_Additive
    translationSpace=BCS_BoneSpace
    alphaBoolBlend {blendOption=Linear}
    alphaScaleBiasClamp {bInterpResult=true, interpSpeedDecreasing=2, interpSpeedIncreasing=3}
    boneToModify {boneName=dyn_hair_l_04}
    rotation {pitch=25}
   }
   n11 = "Transform (Modify) Bone - Bone: dyn_hair_bk_04"(ComponentPose=@n10, Alpha=1, Translation="0,0,0", Rotation="55,0,0", Scale="1,1,1") {
    rotationMode=BMM_Additive
    rotationSpace=BCS_BoneSpace
    translationMode=BMM_Additive
    translationSpace=BCS_BoneSpace
    alphaBoolBlend {blendOption=Linear}
    alphaScaleBiasClamp {bInterpResult=true, interpSpeedDecreasing=2, interpSpeedIncreasing=3}
    boneToModify {boneName=dyn_hair_bk_04}
    rotation {pitch=55}
   }
   n12 = "Transform (Modify) Bone - Bone: dyn_hair_r_04"(ComponentPose=@n11, Alpha=1, Translation="0,0,0", Rotation="25,0,0", Scale="1,1,1") {
    rotationMode=BMM_Additive
    rotationSpace=BCS_BoneSpace
    translationMode=BMM_Additive
    translationSpace=BCS_BoneSpace
    alphaBoolBlend {blendOption=Linear}
    alphaScaleBiasClamp {bInterpResult=true, interpSpeedDecreasing=2, interpSpeedIncreasing=3}
    boneToModify {boneName=dyn_hair_r_04}
    rotation {pitch=25}
   }
   n13 = Component To Local(ComponentPose=@n12)
   n14 = RigidBody!(
    ComponentPose=Local To Component(LocalPose=Copy Pose From Mesh() {bUseAttachedParent=true})
    bAlphaBoolEnabled
    ExternalForce
    ComponentLinearAccScale="0,0,0"
    ComponentLinearVelScale="0,0,0"
   ) -> (Pose) {
    alphaInputType=Bool
    bClampLinearTranslationLimitToRefPose=true
    bOverrideWorldGravity=true
    overridePhysicsAsset=/BRCosmetics/Characters/Player/Female/Medium/Bodies/F_MED_OutlineLove/Meshes/Parts/F_MED_OutlineLove_FaceAcc_Physics
    simulationSpace=BaseBoneSpace
    alphaBoolBlend {blendOption=Linear}
    alphaScaleBiasClamp {interpSpeedDecreasing=4, interpSpeedIncreasing=6}
    baseBoneRef {boneName=spine_05}
    componentAppliedLinearAccClamp=<3500 3500 3500>
    componentLinearAccScale=<0.6 0.6 0.2>
    componentLinearVelScale=<0.5 0.5 0.5>
    overrideWorldGravity {z=-980}
    simSpaceSettings {
     dampingAlpha=0
     maxAngularAcceleration=40
     maxAngularVelocity=8
     maxLinearAcceleration=1500
     maxLinearVelocity=700
     velocityScaleZ=0.1
     worldAlpha=0.1
    }
   }
   n15 = "Transform (Modify) Bone - Bone: dyn_hair_bk_02"(ComponentPose=@n14, Alpha=1, Translation="0,0,0", Rotation="0.000,0.000,-3.000", Scale="1,1,0.9") {
    rotationMode=BMM_Additive
    scaleMode=BMM_Additive
    translationMode=BMM_Additive
    boneToModify {boneName=dyn_hair_bk_02}
    rotation {roll=-3}
    scale {z=0.9}
   }
   n16 = "Transform (Modify) Bone - Bone: dyn_hair_bk_04"(ComponentPose=@n15, Alpha=1, Translation="0,0,0", Rotation="0,0,0", Scale="1,1,0.9") {rotationMode=BMM_Additive, scaleMode=BMM_Additive, translationMode=BMM_Additive, boneToModify {boneName=dyn_hair_bk_04}, scale {z=0.9}}
   n17 = Get BackPackTactJump() -> (BackPackTactJump:double)
   n18 = "Constraint - dyn_hair_r_01 (P)"(ComponentPose=@n16, Alpha=@n17, ConstraintWeights_0=1) {
    alphaBoolBlend {blendOption=Linear}
    boneToModify {boneName=dyn_hair_r_01}
    constraintSetup [{targetBone {boneName=head}}]
    constraintWeights [1]
   }
   n19 = "Constraint - dyn_hair_bk_01 (P)"(ComponentPose=@n18, Alpha=@n17, ConstraintWeights_0=1) {
    alphaBoolBlend {blendOption=Linear}
    boneToModify {boneName=dyn_hair_bk_01}
    constraintSetup [{targetBone {boneName=head}}]
    constraintWeights [1]
   }
   n20 = "Constraint - dyn_hair_l_01 (P)"(ComponentPose=@n19, Alpha=@n17, ConstraintWeights_0=1) {
    alphaBoolBlend {blendOption=Linear}
    boneToModify {boneName=dyn_hair_l_01}
    constraintSetup [{targetBone {boneName=head}}]
    constraintWeights [1]
   }
   n21 = "Constraint - dyn_hair_r_02 (P)"(ComponentPose=@n20, Alpha=@n17, ConstraintWeights_0=1) {
    alphaBoolBlend {blendOption=Linear}
    boneToModify {boneName=dyn_hair_r_02}
    constraintSetup [{targetBone {boneName=neck_01}}]
    constraintWeights [1]
   }
   n22 = "Constraint - dyn_hair_bk_02 (P)"(ComponentPose=@n21, Alpha=@n17, ConstraintWeights_0=1) {
    alphaBoolBlend {blendOption=Linear}
    boneToModify {boneName=dyn_hair_bk_02}
    constraintSetup [{targetBone {boneName=neck_01}}]
    constraintWeights [1]
   }
   n23 = "Constraint - dyn_hair_l_02 (P)"(ComponentPose=@n22, Alpha=@n17, ConstraintWeights_0=1) {
    alphaBoolBlend {blendOption=Linear}
    boneToModify {boneName=dyn_hair_l_02}
    constraintSetup [{targetBone {boneName=neck_01}}]
    constraintWeights [1]
   }
   n24 = "Constraint - dyn_hair_r_03 (P)"(ComponentPose=@n23, Alpha=@n17, ConstraintWeights_0=1) {
    alphaBoolBlend {blendOption=Linear}
    boneToModify {boneName=dyn_hair_r_03}
    constraintSetup [{targetBone {boneName=spine_05}}]
    constraintWeights [1]
   }
   n25 = "Constraint - dyn_hair_r_04 (P)"(ComponentPose=@n24, Alpha=@n17, ConstraintWeights_0=1) {
    alphaBoolBlend {blendOption=Linear}
    boneToModify {boneName=dyn_hair_r_04}
    constraintSetup [{targetBone {boneName=spine_04}}]
    constraintWeights [1]
   }
   n26 = "Constraint - dyn_hair_bk_03 (P)"(ComponentPose=@n25, Alpha=@n17, ConstraintWeights_0=1) {
    alphaBoolBlend {blendOption=Linear}
    boneToModify {boneName=dyn_hair_bk_03}
    constraintSetup [{targetBone {boneName=spine_05}}]
    constraintWeights [1]
   }
   n27 = "Constraint - dyn_hair_bk_04 (P)"(ComponentPose=@n26, Alpha=@n17, ConstraintWeights_0=1) {
    alphaBoolBlend {blendOption=Linear}
    boneToModify {boneName=dyn_hair_bk_04}
    constraintSetup [{targetBone {boneName=spine_04}}]
    constraintWeights [1]
   }
   n28 = "Constraint - dyn_hair_l_03 (P)"(ComponentPose=@n27, Alpha=@n17, ConstraintWeights_0=1) {
    alphaBoolBlend {blendOption=Linear}
    boneToModify {boneName=dyn_hair_l_03}
    constraintSetup [{targetBone {boneName=spine_05}}]
    constraintWeights [1]
   }
   n29 = "Constraint - dyn_hair_l_04 (P)"(ComponentPose=@n28, Alpha=@n17, ConstraintWeights_0=1) {
    alphaBoolBlend {blendOption=Linear}
    boneToModify {boneName=dyn_hair_l_04}
    constraintSetup [{targetBone {boneName=spine_04}}]
    constraintWeights [1]
   }
   n30 = Get SkydivePack() -> (SkydivePack:double)
   n31 = "Constraint - dyn_hair_r_01 (P)"(ComponentPose=@n29, Alpha=@n30, ConstraintWeights_0=1) {
    alphaBoolBlend {blendOption=Linear}
    boneToModify {boneName=dyn_hair_r_01}
    constraintSetup [{targetBone {boneName=head}}]
    constraintWeights [1]
   }
   n32 = "Constraint - dyn_hair_bk_01 (P)"(ComponentPose=@n31, Alpha=@n30, ConstraintWeights_0=1) {
    alphaBoolBlend {blendOption=Linear}
    boneToModify {boneName=dyn_hair_bk_01}
    constraintSetup [{targetBone {boneName=head}}]
    constraintWeights [1]
   }
   n33 = "Constraint - dyn_hair_l_01 (P)"(ComponentPose=@n32, Alpha=@n30, ConstraintWeights_0=1) {
    alphaBoolBlend {blendOption=Linear}
    boneToModify {boneName=dyn_hair_l_01}
    constraintSetup [{targetBone {boneName=head}}]
    constraintWeights [1]
   }
   n34 = "Constraint - dyn_hair_r_02 (P)"(ComponentPose=@n33, Alpha=@n30, ConstraintWeights_0=1) {
    alphaBoolBlend {blendOption=Linear}
    boneToModify {boneName=dyn_hair_r_02}
    constraintSetup [{targetBone {boneName=neck_01}}]
    constraintWeights [1]
   }
   n35 = "Constraint - dyn_hair_bk_02 (P)"(ComponentPose=@n34, Alpha=@n30, ConstraintWeights_0=1) {
    alphaBoolBlend {blendOption=Linear}
    boneToModify {boneName=dyn_hair_bk_02}
    constraintSetup [{targetBone {boneName=neck_01}}]
    constraintWeights [1]
   }
   n36 = "Constraint - dyn_hair_l_02 (P)"(ComponentPose=@n35, Alpha=@n30, ConstraintWeights_0=1) {
    alphaBoolBlend {blendOption=Linear}
    boneToModify {boneName=dyn_hair_l_02}
    constraintSetup [{targetBone {boneName=neck_01}}]
    constraintWeights [1]
   }
   n37 = "Constraint - dyn_hair_r_03 (P)"(ComponentPose=@n36, Alpha=@n30, ConstraintWeights_0=1) {
    alphaBoolBlend {blendOption=Linear}
    boneToModify {boneName=dyn_hair_r_03}
    constraintSetup [{targetBone {boneName=spine_05}}]
    constraintWeights [1]
   }
   n38 = "Constraint - dyn_hair_r_04 (P)"(ComponentPose=@n37, Alpha=@n30, ConstraintWeights_0=1) {
    alphaBoolBlend {blendOption=Linear}
    boneToModify {boneName=dyn_hair_r_04}
    constraintSetup [{targetBone {boneName=spine_04}}]
    constraintWeights [1]
   }
   n39 = "Constraint - dyn_hair_bk_03 (P)"(ComponentPose=@n38, Alpha=@n30, ConstraintWeights_0=1) {
    alphaBoolBlend {blendOption=Linear}
    boneToModify {boneName=dyn_hair_bk_03}
    constraintSetup [{targetBone {boneName=spine_05}}]
    constraintWeights [1]
   }
   n40 = "Constraint - dyn_hair_bk_04 (P)"(ComponentPose=@n39, Alpha=@n30, ConstraintWeights_0=1) {
    alphaBoolBlend {blendOption=Linear}
    boneToModify {boneName=dyn_hair_bk_04}
    constraintSetup [{targetBone {boneName=spine_04}}]
    constraintWeights [1]
   }
   n41 = "Constraint - dyn_hair_l_03 (P)"(ComponentPose=@n40, Alpha=@n30, ConstraintWeights_0=1) {
    alphaBoolBlend {blendOption=Linear}
    boneToModify {boneName=dyn_hair_l_03}
    constraintSetup [{targetBone {boneName=spine_05}}]
    constraintWeights [1]
   }
   n42 = "Constraint - dyn_hair_l_04 (P)"(ComponentPose=@n41, Alpha=@n30, ConstraintWeights_0=1) {
    alphaBoolBlend {blendOption=Linear}
    boneToModify {boneName=dyn_hair_l_04}
    constraintSetup [{targetBone {boneName=spine_04}}]
    constraintWeights [1]
   }
   n43 = "Transform (Modify) Bone - Bone: dyn_hair_r_02"(ComponentPose=@n42, Alpha=@n30, Translation="0,0,5", Rotation="5,5,15", Scale="1,1,1") {
    rotationMode=BMM_Additive
    rotationSpace=BCS_BoneSpace
    translationMode=BMM_Additive
    translationSpace=BCS_BoneSpace
    alphaBoolBlend {blendOption=Linear}
    boneToModify {boneName=dyn_hair_r_02}
    rotation=<rot 5 5 15>
    translation {z=5}
   }
   n44 = "Transform (Modify) Bone - Bone: dyn_hair_l_02"(ComponentPose=@n43, Alpha=@n30, Translation="0,0,5", Rotation="0,-5,-15", Scale="1,1,1") {
    rotationMode=BMM_Additive
    rotationSpace=BCS_BoneSpace
    translationMode=BMM_Additive
    translationSpace=BCS_BoneSpace
    alphaBoolBlend {blendOption=Linear}
    boneToModify {boneName=dyn_hair_l_02}
    rotation {roll=-15, yaw=-5}
    translation {z=5}
   }
   n45 = Component To Local(ComponentPose=@n44)
   n46 = Get isWearBackbling() -> (isWearBackbling:bool)
   n47 = RigidBody!(
    ComponentPose=Local To Component(LocalPose=Blend Poses by bool(BlendPose_0=@n45, BlendPose_1=@n13, bActiveValue=@n46))
    Alpha:float
    OverrideWorldGravity:Vector="0,0,0"
    ExternalForce
   ) -> (Pose) {
    bOverrideWorldGravity=true
    overridePhysicsAsset=/BRCosmetics/Characters/Player/Female/Medium/Bodies/F_MED_OutlineLove/Meshes/Parts/F_MED_OutlineLove_FaceAcc_Front_Physics
    simulationSpace=BaseBoneSpace
    baseBoneRef {boneName=spine_05}
    componentAppliedLinearAccClamp=<3000 3000 3000>
    simSpaceSettings {
     dampingAlpha=0
     maxAngularAcceleration=75
     maxAngularVelocity=40
     maxLinearAcceleration=600
     maxLinearVelocity=250
     velocityScaleZ=0.1
     worldAlpha=0.4
     externalLinearDragV=<0.3 0.3 0.4>
    }
   }
   n48 = "Constraint - dyn_hair_r_03 (PP)"!(ComponentPose=@n47, Alpha=0.5, ConstraintWeights_0=0.5, ConstraintWeights_1:float=0.5) -> (Pose) {
    alphaBoolBlend {blendOption=Linear}
    boneToModify {boneName=dyn_hair_r_03}
    constraintSetup [{targetBone {boneName=dyn_hair_bk_03}}, {targetBone {boneName=clavicle_r}}]
    constraintWeights [1,1]
   }
   n49 = "Constraint - dyn_hair_l_03 (P)"(ComponentPose=@n48, Alpha=0.5, ConstraintWeights_0=1) {
    alphaBoolBlend {blendOption=Linear}
    boneToModify {boneName=dyn_hair_l_03}
    constraintSetup [{targetBone {boneName=dyn_hair_bk_03}}]
    constraintWeights [1]
   }
   n50 = "Constraint - dyn_hair_l_04 (P)"(ComponentPose=@n49, Alpha=0.8, ConstraintWeights_0=1) {
    alphaBoolBlend {blendOption=Linear}
    boneToModify {boneName=dyn_hair_l_04}
    constraintSetup [{targetBone {boneName=dyn_hair_bk_04}}]
    constraintWeights [1]
   }
   n51 = "Constraint - dyn_hair_r_04 (P)"(ComponentPose=@n50, Alpha=0.3, ConstraintWeights_0=1) {
    alphaBoolBlend {blendOption=Linear}
    boneToModify {boneName=dyn_hair_r_04}
    constraintSetup [{targetBone {boneName=dyn_hair_bk_04}}]
    constraintWeights [1]
   }
   n52 = Component To Local(ComponentPose=@n51)
   n53 = "Constraint - dyn_hair_l_02 (P)"(ComponentPose=Local To Component(LocalPose=Copy Pose From Mesh() {bUseAttachedParent=true}), Alpha=1, ConstraintWeights_0=1) {boneToModify {boneName=dyn_hair_l_02}, constraintSetup [{targetBone {boneName=dyn_cape_02}}], constraintWeights [1]}
   n54 = "Constraint - dyn_hair_bk_02 (P)"(ComponentPose=@n53, Alpha=1, ConstraintWeights_0=1) {boneToModify {boneName=dyn_hair_bk_02}, constraintSetup [{targetBone {boneName=dyn_cape_02}}], constraintWeights [1]}
   n55 = "Constraint - dyn_hair_r_02 (P)"(ComponentPose=@n54, Alpha=1, ConstraintWeights_0=1) {boneToModify {boneName=dyn_hair_r_02}, constraintSetup [{targetBone {boneName=dyn_cape_02}}], constraintWeights [1]}
   n56 = "Constraint - dyn_hair_l_03 (P)"(ComponentPose=@n55, Alpha=1, ConstraintWeights_0=1) {boneToModify {boneName=dyn_hair_l_03}, constraintSetup [{targetBone {boneName=dyn_cape_03}}], constraintWeights [1]}
   n57 = "Constraint - dyn_hair_bk_03 (P)"(ComponentPose=@n56, Alpha=1, ConstraintWeights_0=1) {boneToModify {boneName=dyn_hair_bk_03}, constraintSetup [{targetBone {boneName=dyn_cape_03}}], constraintWeights [1]}
   n58 = "Constraint - dyn_hair_r_03 (P)"(ComponentPose=@n57, Alpha=1, ConstraintWeights_0=1) {boneToModify {boneName=dyn_hair_r_03}, constraintSetup [{targetBone {boneName=dyn_cape_03}}], constraintWeights [1]}
   n59 = "Constraint - dyn_hair_l_04 (P)"(ComponentPose=@n58, Alpha=1, ConstraintWeights_0=1) {boneToModify {boneName=dyn_hair_l_04}, constraintSetup [{targetBone {boneName=dyn_cape_03}}], constraintWeights [1]}
   n60 = "Constraint - dyn_hair_bk_04 (P)"(ComponentPose=@n59, Alpha=1, ConstraintWeights_0=1) {boneToModify {boneName=dyn_hair_bk_04}, constraintSetup [{targetBone {boneName=dyn_cape_03}}], constraintWeights [1]}
   n61 = "Constraint - dyn_hair_r_04 (P)"(ComponentPose=@n60, Alpha=1, ConstraintWeights_0=1) {boneToModify {boneName=dyn_hair_r_04}, constraintSetup [{targetBone {boneName=dyn_cape_03}}], constraintWeights [1]}
   n62 = "Transform (Modify) Bone - Bone: dyn_hair_l_02"(ComponentPose=@n61, Alpha=1, Translation="0,0,0", Rotation="0,0,5", Scale="1,1,1") {rotationMode=BMM_Additive, translationMode=BMM_Additive, boneToModify {boneName=dyn_hair_l_02}, rotation {roll=5}}
   n63 = "Constraint - dyn_hair_clav_l_01 (PPP)"!(
    ComponentPose="Transform (Modify) Bone - Bone: dyn_hair_r_02"(
     ComponentPose="Transform (Modify) Bone - Bone: dyn_hair_bk_02"(ComponentPose=@n62, Alpha=1, Translation="0,0,0", Rotation="0,0,5", Scale="1,1,1") {rotationMode=BMM_Additive, translationMode=BMM_Additive, boneToModify {boneName=dyn_hair_bk_02}, rotation {roll=5}}
     Alpha=1
     Translation="0,0,0"
     Rotation="0,0,5"
     Scale="1,1,1"
    ) {rotationMode=BMM_Additive, translationMode=BMM_Additive, boneToModify {boneName=dyn_hair_r_02}, rotation {roll=5}}
    Alpha=1
    ConstraintWeights_0=0.6
    ConstraintWeights_1:float=0.3
    ConstraintWeights_2:float=0.1
   ) -> (Pose) {
    boneToModify {boneName=dyn_hair_clav_l_01}
    constraintSetup [{targetBone {boneName=clavicle_l}}, {targetBone {boneName=spine_05}}, {targetBone {boneName=upperarm_twist_01_l}}]
    constraintWeights [1,1,1]
   }
   n64 = "Constraint - dyn_hair_clav_r_01 (PPP)"!(ComponentPose=@n63, Alpha=1, ConstraintWeights_0=0.6, ConstraintWeights_1:float=0.4, ConstraintWeights_2:float=0) -> (Pose) {
    boneToModify {boneName=dyn_hair_clav_r_01}
    constraintSetup [{targetBone {boneName=clavicle_r}}, {targetBone {boneName=spine_05}}, {targetBone {boneName=upperarm_twist_01_l}}]
    constraintWeights [1,1,1]
   }
   n65 = Get SkydiveFloat() -> (SkydiveFloat:double)
   n66 = "Transform (Modify) Bone - Bone: dyn_hair_l_01"(ComponentPose=@n64, Alpha=@n65, Translation="0,0,0", Rotation="0,0,50", Scale="1,1,1") {rotationMode=BMM_Additive, translationMode=BMM_Additive, boneToModify {boneName=dyn_hair_l_01}, rotation {roll=50}}
   n67 = Component To Local(ComponentPose="Transform (Modify) Bone - Bone: dyn_hair_r_01"(
     ComponentPose="Transform (Modify) Bone - Bone: dyn_hair_bk_01"(ComponentPose=@n66, Alpha=@n65, Translation="0,0,0", Rotation="0,0,50", Scale="1,1,1") {rotationMode=BMM_Additive, translationMode=BMM_Additive, boneToModify {boneName=dyn_hair_bk_01}, rotation {roll=50}}
     Alpha=@n65
     Translation="0,0,0"
     Rotation="0,0,50"
     Scale="1,1,1"
    ) {rotationMode=BMM_Additive, translationMode=BMM_Additive, boneToModify {boneName=dyn_hair_r_01}, rotation {roll=50}})
   n68 = Get bRigidBodyEnabled()
   n69 = Anim Dynamics - Bone: dyn_hair_front_l_01(
    ComponentPose=Local To Component(LocalPose=Blend Poses by bool(BlendPose_0=@n52, BlendPose_1=@n67, bActiveValue=@n68))
    Alpha=1
    ExternalForce="0,0,0"
   ) {
    angularDampingOverride=1
    bOverrideAngularDamping=true
    bOverrideLinearDamping=true
    bUseGravityOverride=true
    linearDampingOverride=0.9999
    simulationSpace=RootRelative
    boundBone {boneName=dyn_hair_front_l_01}
    componentAppliedLinearAccClamp=<3000 3000 3000>
    componentLinearAccScale=<0.1 0.1 0.1>
    componentLinearVelScale=<0.2 0.2 0.2>
    gravityOverride {z=-980}
    physicsBodyDefinitions [{
      boundBone {boneName=dyn_hair_front_l_01}
      boxExtents=<10 10 10>
      constraintSetup {angularLimitsMax=<0 90 30>, angularLimitsMin=<0 -15 0>, angularTarget=<0 0 0>}
      localJointOffset=<1 0 0>
     }]
   }
   n70 = Anim Dynamics - Bone: dyn_hair_front_r_01(ComponentPose=@n69, Alpha=1, ExternalForce="0,0,0") {
    angularDampingOverride=1
    bOverrideAngularDamping=true
    bOverrideLinearDamping=true
    bUseGravityOverride=true
    linearDampingOverride=0.9999
    simulationSpace=RootRelative
    boundBone {boneName=dyn_hair_front_r_01}
    componentAppliedLinearAccClamp=<3000 3000 3000>
    componentLinearAccScale=<0.1 0.1 0.1>
    componentLinearVelScale=<0.2 0.2 0.2>
    gravityOverride {z=-980}
    physicsBodyDefinitions [{
      boundBone {boneName=dyn_hair_front_r_01}
      boxExtents=<10 10 10>
      constraintSetup {angularLimitsMax=<0 90 0>, angularLimitsMin=<0 -5 -30>, angularTarget=<0 0 0>}
      localJointOffset=<1 0 0>
     }]
    relativeSpaceBone {boneName=dyn_hair_clav_r_02}
   }
   n71 = Anim Dynamics - Bone: dyn_hair_front_l_02(ComponentPose=@n70, Alpha=1, ExternalForce="0,0,0") {
    angularDampingOverride=1
    bOverrideAngularDamping=true
    bOverrideLinearDamping=true
    bUseGravityOverride=true
    linearDampingOverride=0.9999
    simulationSpace=RootRelative
    boundBone {boneName=dyn_hair_front_l_02}
    componentAppliedLinearAccClamp=<3000 3000 3000>
    componentLinearAccScale=<0.1 0.1 0.1>
    componentLinearVelScale=<0.2 0.2 0.2>
    gravityOverride {z=-980}
    physicsBodyDefinitions [{
      boundBone {boneName=dyn_hair_front_l_02}
      boxExtents=<10 10 10>
      constraintSetup {angularLimitsMax=<0 90 30>, angularLimitsMin=<0 -15 0>, angularTarget=<0 0 0>}
      localJointOffset=<1 0 0>
     }]
   }
   n72 = Anim Dynamics - Bone: dyn_hair_front_r_02(ComponentPose=@n71, Alpha=1, ExternalForce="0,0,0") {
    angularDampingOverride=1
    bOverrideAngularDamping=true
    bOverrideLinearDamping=true
    bUseGravityOverride=true
    linearDampingOverride=0.9999
    simulationSpace=RootRelative
    boundBone {boneName=dyn_hair_front_r_02}
    componentAppliedLinearAccClamp=<3000 3000 3000>
    componentLinearAccScale=<0.1 0.1 0.1>
    componentLinearVelScale=<0.2 0.2 0.2>
    gravityOverride {z=-980}
    physicsBodyDefinitions [{
      boundBone {boneName=dyn_hair_front_r_02}
      boxExtents=<10 10 10>
      constraintSetup {angularLimitsMax=<0 90 0>, angularLimitsMin=<0 -30 -30>, angularTarget=<0 0 0>}
      localJointOffset=<1 0 0>
     }]
    relativeSpaceBone {boneName=dyn_hair_clav_r_02}
   }
   n73 = Get isIdleInFE() -> (isIdleInFE:double)
   n74 = "Transform (Modify) Bone - Bone: dyn_hair_clav_r_01"(ComponentPose=@n72, Alpha=@n73, Translation="0,0,1", Rotation="0,0,5", Scale="1,1,1") {rotationMode=BMM_Additive, translationMode=BMM_Additive, boneToModify {boneName=dyn_hair_clav_r_01}, rotation {roll=5}, translation {z=1}}
   n75 = "Transform (Modify) Bone - Bone: dyn_hair_clav_r_02"(ComponentPose=@n74, Alpha=@n73, Translation="0,0,0", Rotation="0,0,10", Scale="1,1,1") {rotationMode=BMM_Additive, translationMode=BMM_Additive, boneToModify {boneName=dyn_hair_clav_r_02}, rotation {roll=10}}
   n76 = "Transform (Modify) Bone - Bone: dyn_hair_clav_r_04"(ComponentPose=@n75, Alpha=@n73, Translation="0,0,0", Rotation="0.000,0.000,-25.000", Scale="1,1,1") {rotationMode=BMM_Additive, translationMode=BMM_Additive, boneToModify {boneName=dyn_hair_clav_r_04}, rotation {roll=-25}}
   n77 = "Transform (Modify) Bone - Bone: dyn_hair_clav_l_01"(ComponentPose=@n76, Alpha=@n73, Translation="0,0,1", Rotation="0,0,-5", Scale="1,1,1") {rotationMode=BMM_Additive, translationMode=BMM_Additive, boneToModify {boneName=dyn_hair_clav_l_01}, rotation {roll=-5}, translation {z=1}}
   n78 = "Transform (Modify) Bone - Bone: dyn_hair_clav_l_02"(ComponentPose=@n77, Alpha=@n73, Translation="0,0,0", Rotation="0,0,20", Scale="1,1,1") {rotationMode=BMM_Additive, translationMode=BMM_Additive, boneToModify {boneName=dyn_hair_clav_l_02}, rotation {roll=20}}
   n79 = "Transform (Modify) Bone - Bone: dyn_hair_clav_l_04"(
    ComponentPose="Transform (Modify) Bone - Bone: dyn_hair_clav_l_03"(ComponentPose=@n78, Alpha=@n73, Translation="0,0,0", Rotation="0,0,0", Scale="1,1,1") {rotationMode=BMM_Additive, translationMode=BMM_Additive, boneToModify {boneName=dyn_hair_clav_l_03}}
    Alpha=@n73
    Translation="0,0,0"
    Rotation="0,0,-25"
    Scale="1,1,1"
   ) {rotationMode=BMM_Additive, translationMode=BMM_Additive, boneToModify {boneName=dyn_hair_clav_l_04}, rotation {roll=-25}}
   n80 = "Transform (Modify) Bone - Bone: dyn_hair_bk_04"(ComponentPose=@n79, Alpha=@n73, Translation="0,0,0", Rotation="0.000,10.000,-10.000", Scale="1,1,1") {rotationMode=BMM_Additive, translationMode=BMM_Additive, boneToModify {boneName=dyn_hair_bk_04}, rotation {roll=-10, yaw=10}}
   n81 = Get Emote() -> (Emote:double)
   n82 = "Constraint - dyn_hair_clav_r_01 (PPPP)"!(
    ComponentPose=@n80
    Alpha=@n81
    ConstraintWeights_0=0
    ConstraintWeights_1:float=0.5
    ConstraintWeights_2:float=0.5
    ConstraintWeights_3:float=0
   ) -> (Pose) {
    alphaScaleBiasClamp {bInterpResult=true, interpSpeedDecreasing=7, interpSpeedIncreasing=5}
    boneToModify {boneName=dyn_hair_clav_r_01}
    constraintSetup [{targetBone {boneName=spine_04}}, {targetBone {boneName=spine_05}}, {targetBone {boneName=clavicle_r}}, {targetBone {boneName=upperarm_r}}]
    constraintWeights [1,1,1,1]
   }
   n83 = "Constraint - dyn_hair_clav_r_02 (PPP)"!(ComponentPose=@n82, Alpha=@n81, ConstraintWeights_0=0.4, ConstraintWeights_1:float=0.5, ConstraintWeights_2:float=0.1) -> (Pose) {
    alphaScaleBiasClamp {bInterpResult=true, interpSpeedDecreasing=7, interpSpeedIncreasing=5}
    boneToModify {boneName=dyn_hair_clav_r_02}
    constraintSetup [{targetBone {boneName=spine_05}}, {targetBone {boneName=clavicle_r}}, {targetBone {boneName=spine_04}}]
    constraintWeights [1,1,1]
   }
   n84 = "Constraint - dyn_hair_clav_r_03 (PPPPPP)"!(
    ComponentPose=@n83
    Alpha=@n81
    ConstraintWeights_0=0
    ConstraintWeights_1:float=0
    ConstraintWeights_2:float=0.4
    ConstraintWeights_3:float=0.4
    ConstraintWeights_4:float=0.1
    ConstraintWeights_5:float=0.1
   ) -> (Pose) {
    alphaScaleBiasClamp {bInterpResult=true, interpSpeedDecreasing=7, interpSpeedIncreasing=5}
    boneToModify {boneName=dyn_hair_clav_r_03}
    constraintSetup [
     {targetBone {boneName=spine_02}}
     {targetBone {boneName=spine_03}}
     {targetBone {boneName=spine_04}}
     {targetBone {boneName=spine_05}}
     {targetBone {boneName=clavicle_r}}
     {targetBone {boneName=upperarm_r}}
    ]
    constraintWeights [1,1,1,1,1,1]
   }
   n85 = "Constraint - dyn_hair_clav_r_04 (PPPPPP)"!(
    ComponentPose=@n84
    Alpha=@n81
    ConstraintWeights_0=0
    ConstraintWeights_1:float=0
    ConstraintWeights_2:float=0.9
    ConstraintWeights_3:float=0.1
    ConstraintWeights_4:float=0
    ConstraintWeights_5:float=0
   ) -> (Pose) {
    alphaScaleBiasClamp {bInterpResult=true, interpSpeedDecreasing=7, interpSpeedIncreasing=5}
    boneToModify {boneName=dyn_hair_clav_r_04}
    constraintSetup [
     {targetBone {boneName=spine_02}}
     {targetBone {boneName=spine_03}}
     {targetBone {boneName=spine_04}}
     {targetBone {boneName=spine_05}}
     {targetBone {boneName=clavicle_l}}
     {targetBone {boneName=upperarm_twist_01_l}}
    ]
    constraintWeights [1,1,1,1,1,1]
   }
   n86 = "Constraint - dyn_hair_clav_l_01 (PPPP)"!(
    ComponentPose=@n85
    Alpha=@n81
    ConstraintWeights_0=0
    ConstraintWeights_1:float=0.5
    ConstraintWeights_2:float=0.5
    ConstraintWeights_3:float=0
   ) -> (Pose) {
    alphaScaleBiasClamp {bInterpResult=true, interpSpeedDecreasing=7, interpSpeedIncreasing=5}
    boneToModify {boneName=dyn_hair_clav_l_01}
    constraintSetup [
     {targetBone {boneName=spine_04}}
     {targetBone {boneName=spine_05}}
     {targetBone {boneName=clavicle_l}}
     {targetBone {boneName=upperarm_twist_01_l}}
    ]
    constraintWeights [1,1,1,1]
   }
   n87 = "Constraint - dyn_hair_clav_l_02 (PPP)"!(ComponentPose=@n86, Alpha=@n81, ConstraintWeights_0=0.4, ConstraintWeights_1:float=0.5, ConstraintWeights_2:float=0.1) -> (Pose) {
    alphaScaleBiasClamp {bInterpResult=true, interpSpeedDecreasing=7, interpSpeedIncreasing=5}
    boneToModify {boneName=dyn_hair_clav_l_02}
    constraintSetup [{targetBone {boneName=spine_05}}, {targetBone {boneName=clavicle_l}}, {targetBone {boneName=spine_04}}]
    constraintWeights [1,1,1]
   }
   n88 = "Constraint - dyn_hair_clav_l_03 (PPPPPP)"!(
    ComponentPose=@n87
    Alpha=@n81
    ConstraintWeights_0=0
    ConstraintWeights_1:float=0
    ConstraintWeights_2:float=0.4
    ConstraintWeights_3:float=0.4
    ConstraintWeights_4:float=0.1
    ConstraintWeights_5:float=0.1
   ) -> (Pose) {
    alphaScaleBiasClamp {bInterpResult=true, interpSpeedDecreasing=7, interpSpeedIncreasing=5}
    boneToModify {boneName=dyn_hair_clav_l_03}
    constraintSetup [
     {targetBone {boneName=spine_02}}
     {targetBone {boneName=spine_03}}
     {targetBone {boneName=spine_04}}
     {targetBone {boneName=spine_05}}
     {targetBone {boneName=clavicle_l}}
     {targetBone {boneName=upperarm_twist_01_l}}
    ]
    constraintWeights [1,1,1,1,1,1]
   }
   n89 = "Constraint - dyn_hair_clav_l_04 (PPPPPP)"!(
    ComponentPose=@n88
    Alpha=@n81
    ConstraintWeights_0=0
    ConstraintWeights_1:float=0
    ConstraintWeights_2:float=0.9
    ConstraintWeights_3:float=0.1
    ConstraintWeights_4:float=0
    ConstraintWeights_5:float=0
   ) -> (Pose) {
    alphaScaleBiasClamp {bInterpResult=true, interpSpeedDecreasing=7, interpSpeedIncreasing=5}
    boneToModify {boneName=dyn_hair_clav_l_04}
    constraintSetup [
     {targetBone {boneName=spine_02}}
     {targetBone {boneName=spine_03}}
     {targetBone {boneName=spine_04}}
     {targetBone {boneName=spine_05}}
     {targetBone {boneName=clavicle_l}}
     {targetBone {boneName=upperarm_twist_01_l}}
    ]
    constraintWeights [1,1,1,1,1,1]
   }
   n90 = Control Rig(Source=Component To Local(ComponentPose=@n89)) {controlRigAssetReference {blueprintRigClass=/BRCosmetics/Characters/Player/Female/Medium/Bodies/F_MED_OutlineLove_Blur/Meshes/F_MED_OutlineLove_Blur_FaceAcc_ctrlRig.F_MED_OutlineLove_Blur_FaceAcc_ctrlRig_C}}
   Output Pose(Result=@n90) {layerGroup=None}
  ]}
 EventGraph {nodes [
   Event OnBeginSkydiving()
   > Set SkydiveFloat(SkydiveFloat:double=1) -> (Output_Get:double)
   Event OnEndJumping()
   > Set JumpingFloat(JumpingFloat:double=0) -> (Output_Get:double)
   Event OnEndFalling()
   > Set JumpingFloat(JumpingFloat=0)
   Event BlueprintInitializeAnimation()
   > Set Body_Mesh(Body_Mesh=GetOwnerSkeletalMeshComponent())
   Event OnEndSkydiving()
   > Set SkydiveFloat(SkydiveFloat=0)
   Event OnBeginPlayingEmote() {is=title K2Node_Event}
   > Set Emote(Emote:double=0.5) -> (Output_Get:double)
   Event OnBeginFalling()
   > Set JumpingFloat(JumpingFloat=1)
   Event OnEndPlayingEmote() {is=title K2Node_Event}
   > Set Emote(Emote=0)
   n1 = Event OnCharacterCustomizationLoaded()
   n2 = CheckForBackbling()
   > Set isWearBackbling(isWearBackbling:bool=@n2) -> (Output_Get:bool)
   > Set isWearCape(isWearCape:bool=@n2.bIsBackblingLong) -> (Output_Get:bool)
   > Set Body(Body:Skeletal Mesh Component Budgeted Object Reference=Get Array Item with EFort Custom Part Type(Array=@n1.CustomMeshParts)) -> (Output_Get:Skeletal Mesh Component Budgeted Object Reference)
   > Set Backbling(Backbling:Skeletal Mesh Component Budgeted Object Reference=Get Array Item with EFort Custom Part Type(Array=@n1.CustomMeshParts, PartType=Backpack)) -> (Output_Get:Skeletal Mesh Component Budgeted Object Reference)
   Event BlueprintUpdateAnimation()
   > n3 = Sequence!(execute) -> (then_0, then_1, then_2:Exec, then_3:Exec, then_4:Exec)
   > n4 = "Gravity Override State Function (Parameter Asset) (Nativized)"(
    CharacterPartAnimInstance=Self-Reference()
    GameTimeInSeconds=GetGameTimeInSeconds()
    Parameters=/BRCosmetics/Characters/Player/Female/Medium/Bodies/F_MED_OutlineLove/Meshes/Parts/F_MED_OutlineLoveFA_Parameters
   )
   > Set GravityOverride(GravityOverride=@n4)
   n5 = AND Boolean(A=Get isWearBackbling(), B=NOT Boolean(A=Get isWearCape() -> (isWearCape:bool)))
   n6 = Get bIsTacticalSprinting()
   Set BackPackTactJump(execute=@n3.then_1, BackPackTactJump:double=SelectFloat(A=1, B=0, bPickA=AND Boolean(A=@n6, B=@n5))) -> (Output_Get:double)
   n7 = Get bIsSkyDiving()
   n8 = Get isWearBackbling()
   > Set SkydivePack(SkydivePack:double=SelectFloat(A=1, B=0, bPickA=AND Boolean(A=@n7, B=@n8))) -> (Output_Get:double)
   n9 = NOT Boolean(A=Get bIsMoving2D())
   n10 = NOT Boolean(A=Get bIsSkyDiving())
   n11 = Get bIsFalling()
   > Set JumpingNotMoving2D(JumpingNotMoving2D:double=SelectFloat(A=1, B=0, bPickA=AND Boolean!(A=@n11, B=@n9, C:bool=@n10) -> (ReturnValue))) -> (Output_Get:double)
   n12 = Cast To CustomCharacterPartAnimInstance!(execute=@n3.then_2, Object=GetAnimInstance(self=Get Backbling() -> (Backbling:Skeletal Mesh Component Budgeted Object Reference))) -> (then, CastFailed, AsCustom Character Part Anim Instance:Custom Character Part Anim Instance Object Reference) {is=title K2Node_DynamicCast}
   > Set Cape Cloth Disabled(Cape Cloth Disabled:bool=NOT Boolean(A=Get bClothEnabled!(self:Fort Anim Instance Object Reference=@n12) -> (bClothEnabled))) -> (Output_Get:bool)
   n13 = Branch(
    execute=@n3.then_3
    Condition=OR Boolean!(A=Get bIsSkyDiving(), B=Get bIsCrouching(), C:bool=Get bIsPlayingEmote(), D:bool=Get isWearBackbling()) -> (ReturnValue)
   )
   > Set isIdle(isIdle:double=0) -> (Output_Get:double)
   > Set isIdleInFE(isIdleInFE:double=0) -> (Output_Get:double)
   n14 = float + float!(
    A=Get strafeLeft float() -> (strafeLeft float:double)
    B=Get strafeRight float() -> (strafeRight float:double)
    C:double=Get moveForward() -> (moveForward:double)
    D:double=Get moveBackward() -> (moveBackward:double)
   ) -> (ReturnValue)
   Set isIdle(execute=@n13.else, isIdle=Select(Option 0:double=1, Option 1:double=0, Index="float > float"(A=@n14)) -> (ReturnValue:double))
   > n15 = Branch(Condition=Get bIsInFrontEnd())
   > Set isIdleInFE(isIdleInFE=Get isIdle() -> (isIdle:double))
   Set isIdleInFE(execute=@n15.else, isIdleInFE=0)
   Event OnBeginJumping()
   > Set JumpingFloat(JumpingFloat=1)
   Reroute Node(InputPin:Wildcard) -> (OutputPin:Wildcard)
   Get bIsJumping()
  ]}
]
variables [
 Backbling {type=SkeletalMeshComponentBudgeted}
 BackPackTactJump {type=number}
 Body {type=SkeletalMeshComponentBudgeted}
 Body_Mesh {type=SkeletalMeshComponent}
 BwdMovement {type=number}
 Cape Cloth Disabled {type=boolean}
 CrouchMove {type=number}
 Emote {type=number}
 GravityOverride {type=Vector}
 Is Backbling A Cape {type=boolean}
 isIdle {type=number}
 isIdleInFE {type=number}
 isWearBackbling {type=boolean}
 isWearCape {type=boolean}
 JumpingFloat {type=number}
 JumpingNotMoving2D {type=number}
 moveBackward {type=number}
 moveForward {type=number}
 RBAN Settings Out {type=FortRigidBodyAnimNodeParameters}
 SkydiveFloat {type=number}
 SkydivePack {type=number}
 strafeLeft float {type=number}
 strafeRight float {type=number}
 Targeting {type=number}
 TrailControllerRot {type=Rotator}
]
