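# Physics asset for the cape colliders on the face accessory part of the OutlineLove_Blur medium female character cosmetic, likely keeping a cape clear of the hair.
# A large zero-length capsule, effectively an offset sphere, sits on the dyn_hair_upper_cape_collision bone, though its shape name refers to a lower ear cape collider.
# The head bone carries only a tiny capsule.
# No constraints link the two bodies.
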PhysicsAsset /BRCosmetics/Characters/Player/Female/Medium/Bodies/F_MED_OutlineLove_Blur/Meshes/Parts/F_MED_OutlineLove_Blur_FaceAcc_CapeColliders_Physics
bodies {
 dyn_hair_upper_cape_collision [{
   shapeType=Capsule
   length=0
   radius=35.255
   shapeName=dyn_ear_lower_cape_collision_capsule
   center=<2.7149 -11.5836 0>
   rotation=<rot 0 -110.1696 0>
  }]
 head [{shapeType=Capsule, length=1, radius=1, rotation=<rot -0.0515 -90 -90.2064>}]
}
constraints []
